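# Niagara idle effect for the NutmegMayo backpack, giving off flickering orange ember sparks and warm light at mesh sockets.
# Near-identical ember emitters sample separate sockets (FXSocket_01, FXSocket_02), spraying short-lived sparks with drag and gravity.
# A repeated cyclic noise curve, remapped by the system's FlickerCurveRemapMin, modulates both spark spawn probability and light brightness.
# Spawn groups separate sparks from a burst particle lasting the loop, supporting sprite, light and component renderers.
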NiagaraSystem /BRCosmetics/Accessories/FORT_Backpacks/Backpack_F_MED_NutmegMayo/FX/NS_Backpack_NutmegMayo_Idle
dependencies {
 usedDataInterfaces [&NiagaraDataInterfaceCurve, &NiagaraDataInterfaceSkeletalMesh]
 usedDynamicInputs []
 usedModules [
  &SystemState
  &self:SystemSpawnScript.SystemScriptSource.SystemScriptGraph.NiagaraNodeAssignment_0.SetVariables_#1
  &self:SystemSpawnScript.SystemScriptSource.SystemScriptGraph.NiagaraNodeAssignment_1.SetVariables_#2
 ]
}
emitters {
 Embers001 {
  emitterUpdateScript [
   &EmitterState(Life Cycle Mode=::NewEnumerator1, Loop Duration=$System.CustomLoopDuration, UseLoopDelay=false)
   &SpawnBurst_Instantaneous(Spawn Group=1, Use Spawn Probability=false, UseLoopCountLimit=false) {module=SpawnBurst_Instantaneous001}
   &SpawnRate(
    Spawn Probability=&RemapRange(
     Input Value To Remap=&FloatFromCurve(
      FloatCurve=DataInterface() {propertyValues {Curve {
         postInfinityExtrap=RCCE_Cycle
         preInfinityExtrap=RCCE_Cycle
         keys [
          <at 0 1 RCIM_Cubic>
          {arriveTangent=0.5283, interpMode=RCIM_Cubic, leaveTangent=0.5283, time=0.0746, value=0.875}
          {arriveTangent=-0.0227, interpMode=RCIM_Cubic, leaveTangent=-0.0227, tangentMode=RCTM_User, time=0.1682, value=1.0888}
          {arriveTangent=-1.3938, interpMode=RCIM_Cubic, leaveTangent=-1.3938, time=0.2019, value=0.6577}
          {arriveTangent=1.4729, interpMode=RCIM_Cubic, leaveTangent=1.4729, time=0.2748, value=0.9402}
          {arriveTangent=0.905, interpMode=RCIM_Cubic, leaveTangent=0.905, time=0.3464, value=0.8705}
          {arriveTangent=-1.2159, interpMode=RCIM_Cubic, leaveTangent=-1.2159, time=0.4074, value=1.0603}
          {arriveTangent=-1.5422, interpMode=RCIM_Cubic, leaveTangent=-1.5422, time=0.4365, value=0.7609}
          {arriveTangent=-0.8529, interpMode=RCIM_Cubic, leaveTangent=-0.8529, time=0.4871, value=0.9373}
          {arriveTangent=0.3027, interpMode=RCIM_Cubic, leaveTangent=0.3027, time=0.5331, value=0.6785}
          {arriveTangent=1.4034, interpMode=RCIM_Cubic, leaveTangent=1.4034, time=0.6112, value=0.9748}
          {arriveTangent=1.5913, interpMode=RCIM_Cubic, leaveTangent=1.5913, time=0.69, value=0.8986}
          {arriveTangent=-0.751, interpMode=RCIM_Cubic, leaveTangent=-0.751, time=0.7536, value=1.2014}
          {arriveTangent=-2.595, interpMode=RCIM_Cubic, leaveTangent=-2.595, time=0.7995, value=0.8164}
          {arriveTangent=-0.8671, interpMode=RCIM_Cubic, leaveTangent=-0.8671, time=0.8674, value=0.9059}
          {arriveTangent=-0.3671, interpMode=RCIM_Cubic, leaveTangent=-0.3671, time=0.9138, value=0.7173}
          {arriveTangent=-0.6714, interpMode=RCIM_Cubic, leaveTangent=-0.6714, time=0.9899, value=0.8609}
          {arriveTangent=0.7195, interpMode=RCIM_Cubic, leaveTangent=0.7195, time=1.0372, value=0.6345}
          {arriveTangent=1.5649, interpMode=RCIM_Cubic, leaveTangent=1.5649, time=1.099, value=0.9394}
          {arriveTangent=0.3682, interpMode=RCIM_Cubic, leaveTangent=0.3682, time=1.1296, value=0.7792}
          {arriveTangent=1.199, interpMode=RCIM_Cubic, leaveTangent=1.199, time=1.2078, value=0.9795}
          {arriveTangent=0.6999, interpMode=RCIM_Cubic, leaveTangent=0.6999, time=1.2281, value=0.8972}
          {arriveTangent=0.8013, interpMode=RCIM_Cubic, leaveTangent=0.8013, time=1.329, value=1.0643}
          {arriveTangent=0.4649, interpMode=RCIM_Cubic, leaveTangent=0.4649, time=1.3719, value=1.0125}
          {arriveTangent=-3.0415, interpMode=RCIM_Cubic, leaveTangent=-3.0415, time=1.414, value=1.1038}
          {arriveTangent=-0.0928, interpMode=RCIM_Cubic, leaveTangent=-0.0928, tangentMode=RCTM_User, time=1.4501, value=0.7746}
          {arriveTangent=0.7682, interpMode=RCIM_Cubic, leaveTangent=0.7682, time=1.5242, value=0.9899}
          {arriveTangent=0.5284, interpMode=RCIM_Cubic, leaveTangent=0.5284, time=1.5566, value=0.8565}
          {arriveTangent=-0.1056, interpMode=RCIM_Cubic, leaveTangent=-0.1056, time=1.7244, value=1.0957}
          {arriveTangent=-2.1193, interpMode=RCIM_Cubic, leaveTangent=-2.1193, time=1.7619, value=0.8348}
          {arriveTangent=-1.2854, interpMode=RCIM_Cubic, leaveTangent=-1.2854, time=1.7959, value=0.9442}
          {arriveTangent=-0.1279, interpMode=RCIM_Cubic, leaveTangent=-0.1279, time=1.836, value=0.7395}
          {arriveTangent=1.0187, interpMode=RCIM_Cubic, leaveTangent=1.0187, time=1.9343, value=0.9265}
          {arriveTangent=1.0782, interpMode=RCIM_Cubic, leaveTangent=1.0782, time=1.9912, value=0.8977}
          {arriveTangent=-2.5847, interpMode=RCIM_Cubic, leaveTangent=-2.5847, time=2.0488, value=1.0499}
          {arriveTangent=-0.6379, interpMode=RCIM_Cubic, leaveTangent=-0.6379, time=2.0828, value=0.6611}
          {arriveTangent=0.4353, interpMode=RCIM_Cubic, leaveTangent=0.4353, time=2.2158, value=0.9434}
          {arriveTangent=-0.3447, interpMode=RCIM_Cubic, leaveTangent=-0.3447, time=2.2673, value=0.7414}
          {arriveTangent=0.6603, interpMode=RCIM_Cubic, leaveTangent=0.6603, time=2.3258, value=0.9055}
          {arriveTangent=0.3801, interpMode=RCIM_Cubic, leaveTangent=0.3801, time=2.3971, value=0.8271}
          {arriveTangent=0.6687, interpMode=RCIM_Cubic, leaveTangent=0.6687, time=2.4654, value=0.9586}
          {arriveTangent=1.2527, interpMode=RCIM_Cubic, leaveTangent=1.2527, time=2.5143, value=0.9055}
          {arriveTangent=-0.515, interpMode=RCIM_Cubic, leaveTangent=-0.515, time=2.5566, value=1.0728}
          {arriveTangent=-1.6201, interpMode=RCIM_Cubic, leaveTangent=-1.6201, time=2.597, value=0.8629}
          {arriveTangent=-0.9669, interpMode=RCIM_Cubic, leaveTangent=-0.9669, time=2.6456, value=0.9286}
          {arriveTangent=0.1009, interpMode=RCIM_Cubic, leaveTangent=0.1009, time=2.6921, value=0.771}
          {arriveTangent=0.0929, interpMode=RCIM_Cubic, leaveTangent=0.0929, time=2.7509, value=0.9392}
          {arriveTangent=0.2939, interpMode=RCIM_Cubic, leaveTangent=0.2939, time=2.8654, value=0.7871}
          {arriveTangent=1.7227, interpMode=RCIM_Cubic, leaveTangent=1.7227, time=2.9271, value=0.991}
          {arriveTangent=0.1236, interpMode=RCIM_Cubic, leaveTangent=0.1236, time=2.957, value=0.9449}
          <at 3 1 RCIM_Cubic>
         ]
        }}}
      CurveIndex=&Add_Float(A=$Emitter.NormalizedLoopAge, B=0)
     )
     Input Min=$System.FlickerCurveRemapMin
     Input Max=1.2
     Output Min=0
     Output Max=0.5
    )
    SpawnRate=&Multiply_Float(A=1, B=100)
    Use Spawn Probability=true
   )
  ]
  particleSpawnScript [
   /SetVariables(Particles.SpawnGroup=$Emitter.SpawnGroup) {isSetParameters=true, module=SetVariables_#3}
   /SetVariables(Particles.SparkIntensity:NiagaraFloat=&FloatFromCurve(
     FloatCurve=DataInterface() {propertyValues {Curve {keys [
         <at 0 1 RCIM_Cubic>
         {arriveTangent=-2.5317, interpMode=RCIM_Cubic, leaveTangent=-2.5317, tangentMode=RCTM_User, time=0.6, value=0.5}
         <at 1 0 RCIM_Cubic>
        ]}}}
     CurveIndex=&RandomRangeFloat(Minimum=0, Maximum=1)
    )) {isSetParameters=true, module=SetVariables_#4}
   &InitializeParticle(
    Color=&MultiplyLinearColor(
     Linear Color=&LerpLinearColors(StartColor=<rgba 10 1.5225 0.5208 1>, EndColor=<rgba 25 16.8856 2.9948 1>, LerpFactor=$Particles.SparkIntensity)
     Scale Factor=<1 0.5 0.1 1>
    )
    Lifetime=&MaskFloatBySpawnGroup(
     Masked Float=&RandomRangeFloat(Minimum=0.17, Maximum=0.25)
     Passthrough Float=$System.CustomLoopDuration
    )
    Sprite Size Mode=::NewEnumerator1
    Uniform Sprite Size Max=5
    Uniform Sprite Size Min=6
    UsePositionOffset=false
    Write Material Random=true
   )
   &SkeletalMeshLocation(
    MaskSpawnGroup=false
    Mesh Sampling Type=::NewEnumerator5
    Skeletal Mesh=DataInterface() {propertyValues {
      DefaultMesh=null
      ExcludeBoneName=root
      OptionalComponentClass=null
      PreviewMesh=/BRCosmetics/Accessories/FORT_Backpacks/Backpack_F_MED_NutmegMayo/Meshes/F_MED_NutmegMayo_Pack
      SkinningMode=SkinOnTheFly
      SoftSourceActor=null
      SourceMode=Default
      UvSetIndex=0
      WholeMeshLOD=-1
      bExcludeBone=false
      bReadDeformedGeometry=true
      bRequireCurrentFrameData=true
      ComponentTags []
      FilteredBones []
      FilteredSockets [FXSocket_01]
      MeshUserParameter {parameter {name=None, typeDefHandle {registeredTypeIndex=1}, varData []}}
      SamplingRegions []
     }}
    Socket Sampling Mode=::NewEnumerator5
   )
   &ShapeLocation(
    MaskSpawnGroup=false
    Offset Coordinate Space=::Local
    Offset Mode=::NewEnumerator0
    Override Seed=false
    Sphere Radius=0
    UseOwnerScale=true
   )
   &AddVelocity(
    Module.Use Falloff Distance=false
    Origin Offset Coordinate Space=::World
    Override Seed=false
    Velocity Mode=::NewEnumerator1
    Velocity Origin=$Particles.SkeletalMeshLocation.SampledPosition
    Velocity Speed=&Multiply_Float(A=&RandomRangeFloat(Minimum=20, Maximum=80), B=1)
   ) {module=AddVelocity001}
   &AddVelocity(
    Cone Angle=80
    Cone Axis=<0 0 -1>
    Module.UseConeFalloff=false
    Module.UseVelDistribution=false
    Override Seed=false
    Velocity Mode=::NewEnumerator2
    Velocity Speed=&RandomRangeFloat(Minimum=10, Maximum=80)
   ) {module=AddVelocity002}
   &ApplyInitialForces()
   &FN_Gen5_Light(
    Attenutation=15
    Color:LinearColor=<rgba 1 0.6 0.2 1>
    Intensity:NiagaraFloat=&Multiply_Float(
     A=&Multiply_Float(
      A=150
      B=&MakeCustomFloatFromBool(
       Bool=&SetBoolByIntComparison(A=$Engine.Emitter.NumParticles, B=3, Comparison Type=::NewEnumerator0)
       True Float=1
       False Float=0
      )
     )
     B=&RemapRange(
      Input Value To Remap=&FloatFromCurve(
       FloatCurve=DataInterface() {propertyValues {
         bOptimizeLUT=false
         Curve {
          postInfinityExtrap=RCCE_Cycle
          preInfinityExtrap=RCCE_Cycle
          keys [
           <at 0 1 RCIM_Cubic>
           {arriveTangent=0.5283, interpMode=RCIM_Cubic, leaveTangent=0.5283, time=0.0746, value=0.875}
           {arriveTangent=-0.0227, interpMode=RCIM_Cubic, leaveTangent=-0.0227, tangentMode=RCTM_User, time=0.1682, value=1.0888}
           {arriveTangent=-1.3938, interpMode=RCIM_Cubic, leaveTangent=-1.3938, time=0.2019, value=0.6577}
           {arriveTangent=1.4729, interpMode=RCIM_Cubic, leaveTangent=1.4729, time=0.2748, value=0.9402}
           {arriveTangent=0.905, interpMode=RCIM_Cubic, leaveTangent=0.905, time=0.3464, value=0.8705}
           {arriveTangent=-1.2159, interpMode=RCIM_Cubic, leaveTangent=-1.2159, time=0.4074, value=1.0603}
           {arriveTangent=-1.5422, interpMode=RCIM_Cubic, leaveTangent=-1.5422, time=0.4365, value=0.7609}
           {arriveTangent=-0.8529, interpMode=RCIM_Cubic, leaveTangent=-0.8529, time=0.4871, value=0.9373}
           {arriveTangent=0.3027, interpMode=RCIM_Cubic, leaveTangent=0.3027, time=0.5331, value=0.6785}
           {arriveTangent=1.4034, interpMode=RCIM_Cubic, leaveTangent=1.4034, time=0.6112, value=0.9748}
           {arriveTangent=1.5913, interpMode=RCIM_Cubic, leaveTangent=1.5913, time=0.69, value=0.8986}
           {arriveTangent=-0.751, interpMode=RCIM_Cubic, leaveTangent=-0.751, time=0.7536, value=1.2014}
           {arriveTangent=-2.595, interpMode=RCIM_Cubic, leaveTangent=-2.595, time=0.7995, value=0.8164}
           {arriveTangent=-0.8671, interpMode=RCIM_Cubic, leaveTangent=-0.8671, time=0.8674, value=0.9059}
           {arriveTangent=-0.3671, interpMode=RCIM_Cubic, leaveTangent=-0.3671, time=0.9138, value=0.7173}
           {arriveTangent=-0.6714, interpMode=RCIM_Cubic, leaveTangent=-0.6714, time=0.9899, value=0.8609}
           {arriveTangent=0.7195, interpMode=RCIM_Cubic, leaveTangent=0.7195, time=1.0372, value=0.6345}
           {arriveTangent=1.5649, interpMode=RCIM_Cubic, leaveTangent=1.5649, time=1.099, value=0.9394}
           {arriveTangent=0.3682, interpMode=RCIM_Cubic, leaveTangent=0.3682, time=1.1296, value=0.7792}
           {arriveTangent=1.199, interpMode=RCIM_Cubic, leaveTangent=1.199, time=1.2078, value=0.9795}
           {arriveTangent=0.6999, interpMode=RCIM_Cubic, leaveTangent=0.6999, time=1.2281, value=0.8972}
           {arriveTangent=0.8013, interpMode=RCIM_Cubic, leaveTangent=0.8013, time=1.329, value=1.0643}
           {arriveTangent=0.4649, interpMode=RCIM_Cubic, leaveTangent=0.4649, time=1.3719, value=1.0125}
           {arriveTangent=-3.0415, interpMode=RCIM_Cubic, leaveTangent=-3.0415, time=1.414, value=1.1038}
           {arriveTangent=-0.0928, interpMode=RCIM_Cubic, leaveTangent=-0.0928, tangentMode=RCTM_User, time=1.4501, value=0.7746}
           {arriveTangent=0.7682, interpMode=RCIM_Cubic, leaveTangent=0.7682, time=1.5242, value=0.9899}
           {arriveTangent=0.5284, interpMode=RCIM_Cubic, leaveTangent=0.5284, time=1.5566, value=0.8565}
           {arriveTangent=-0.1056, interpMode=RCIM_Cubic, leaveTangent=-0.1056, time=1.7244, value=1.0957}
           {arriveTangent=-2.1193, interpMode=RCIM_Cubic, leaveTangent=-2.1193, time=1.7619, value=0.8348}
           {arriveTangent=-1.2854, interpMode=RCIM_Cubic, leaveTangent=-1.2854, time=1.7959, value=0.9442}
           {arriveTangent=-0.1279, interpMode=RCIM_Cubic, leaveTangent=-0.1279, time=1.836, value=0.7395}
           {arriveTangent=1.0187, interpMode=RCIM_Cubic, leaveTangent=1.0187, time=1.9343, value=0.9265}
           {arriveTangent=1.0782, interpMode=RCIM_Cubic, leaveTangent=1.0782, time=1.9912, value=0.8977}
           {arriveTangent=-2.5847, interpMode=RCIM_Cubic, leaveTangent=-2.5847, time=2.0488, value=1.0499}
           {arriveTangent=-0.6379, interpMode=RCIM_Cubic, leaveTangent=-0.6379, time=2.0828, value=0.6611}
           {arriveTangent=0.4353, interpMode=RCIM_Cubic, leaveTangent=0.4353, time=2.2158, value=0.9434}
           {arriveTangent=-0.3447, interpMode=RCIM_Cubic, leaveTangent=-0.3447, time=2.2673, value=0.7414}
           {arriveTangent=0.6603, interpMode=RCIM_Cubic, leaveTangent=0.6603, time=2.3258, value=0.9055}
           {arriveTangent=0.3801, interpMode=RCIM_Cubic, leaveTangent=0.3801, time=2.3971, value=0.8271}
           {arriveTangent=0.6687, interpMode=RCIM_Cubic, leaveTangent=0.6687, time=2.4654, value=0.9586}
           {arriveTangent=1.2527, interpMode=RCIM_Cubic, leaveTangent=1.2527, time=2.5143, value=0.9055}
           {arriveTangent=-0.515, interpMode=RCIM_Cubic, leaveTangent=-0.515, time=2.5566, value=1.0728}
           {arriveTangent=-1.6201, interpMode=RCIM_Cubic, leaveTangent=-1.6201, time=2.597, value=0.8629}
           {arriveTangent=-0.9669, interpMode=RCIM_Cubic, leaveTangent=-0.9669, time=2.6456, value=0.9286}
           {arriveTangent=0.1009, interpMode=RCIM_Cubic, leaveTangent=0.1009, time=2.6921, value=0.771}
           {arriveTangent=0.0929, interpMode=RCIM_Cubic, leaveTangent=0.0929, time=2.7509, value=0.9392}
           {arriveTangent=0.2939, interpMode=RCIM_Cubic, leaveTangent=0.2939, time=2.8654, value=0.7871}
           {arriveTangent=1.7227, interpMode=RCIM_Cubic, leaveTangent=1.7227, time=2.9271, value=0.991}
           {arriveTangent=0.1236, interpMode=RCIM_Cubic, leaveTangent=0.1236, time=2.957, value=0.9449}
           <at 3 1 RCIM_Cubic>
          ]
         }
        }}
       CurveIndex=$Emitter.NormalizedLoopAge
      )
      Input Min=&Add_Float(A=$System.FlickerCurveRemapMin, B=0)
      Input Max=1.2
      Output Min=0
      Output Max=1
     )
    )
   ) {module=FN_Gen5_Light001}
  ]
  particleUpdateScript [
   &ParticleState()
   &ScaleSpriteSize(Scale Factor=&Multiply_Vector2D_ByFloat(Float=0.1, Vector2D=<1 0.7>), Scale Sprite Size Mode=::NewEnumerator2)
   &SkeletalMeshLocation(
    MaskSpawnGroup=false
    Mesh Sampling Type=::NewEnumerator5
    Position Sampling=::NewEnumerator2
    Skeletal Mesh=DataInterface() {propertyValues {
      DefaultMesh=null
      ExcludeBoneName=root
      OptionalComponentClass=null
      PreviewMesh=&F_MED_NutmegMayo_Pack
      SkinningMode=SkinOnTheFly
      SoftSourceActor=null
      SourceMode=Default
      UvSetIndex=0
      WholeMeshLOD=-1
      bExcludeBone=false
      bReadDeformedGeometry=true
      bRequireCurrentFrameData=true
      ComponentTags []
      FilteredBones []
      FilteredSockets [FXSocket_01]
      MeshUserParameter {parameter {name=None, typeDefHandle {registeredTypeIndex=1}, varData []}}
      SamplingRegions []
     }}
    Socket Sampling Mode=::NewEnumerator5
   ) {module=SkeletalMeshLocation001}
   &FN_Sprite_Attributes(
    Camera Offset=15
    Color=&MultiplyLinearColor(
     Linear Color=$Particles.Color
     Scale Factor=&MakeVector4(
      X=1
      Y=1
      Z=1
      W=&Multiply_Float(
       A=0.1
       B=&FloatFromCurve(FloatCurve=DataInterface() {propertyValues {Curve {keys [
            {interpMode=RCIM_Cubic, tangentMode=RCTM_User, time=0, value=0}
            {interpMode=RCIM_Cubic, tangentMode=RCTM_User, time=0.5, value=1}
            {tangentMode=RCTM_User, time=1, value=0}
           ]}}})
      )
     )
    )
    Create Camera Offset Attribute=true
    Create Color Attribute=true
    Create Position Attribute=true
    Create SpriteSize Attribute=true
    Position:NiagaraPosition=$Particles.SkeletalMeshLocation001.SampledPosition
    SpriteSize=&Multiply_Vector2D_ByFloat(Float=0.05, Vector2D=&Multiply_Vector2D(A=<400 400>, B=$Particles.SpriteSize))
    Write to Module Namespace=true
   )
   &Drag(Use Linear Drag=true, Use Rotational Drag=false)
   &GravityForce(Gravity=&Multiply_Vector(A=<0 0 -980>, B=<1 1 0.35>))
   &InheritSourceMovement(ApplyForce=false, ApplyPosition=true, ApplyPositionFalloff=false)
   &SolveForcesAndVelocity(Clamp Velocity=false, Limit Acceleration=false)
   &ScaleSpriteSizeBySpeed(Max Scale Factor=<1 5>, Min Scale Factor=<1 -5>, Velocity Threshold=50)
   /SetVariables(Particles.FlareDynamic:Vector4f=<0.1 1 0 0>) {isSetParameters=true, module=SetVariables_#5}
   /SetVariables(Particles.VisibilityTag=$Particles.Initial.SpawnGroup) {isSetParameters=true, module=SetVariables_#6}
   &InterpolateOverTime(
    Rate Of Change=5
    Target Float Value:NiagaraFloat=&MakeCustomFloatFromBool(
     Bool=&SetBoolByIntComparison(A=$Engine.Emitter.NumParticles, B=2, Comparison Type=::NewEnumerator0)
     True Float=1
     False Float=0
    )
    Use Set Interpolation Rate=true
   )
   &FN_Light_Attributes(
    Color=$Particles.Color
    Create Color Attribute=true
    Create Exponent Attribute=true
    Create Position Attribute=true
    Create Radius Attribute=true
    Enable Particle Light Enable=true
    Exponent=1.5
    Light Enabled=true
    Position=$Particles.FN_Sprite_Attributes.Position
    Radius=&Multiply_Float(A=15, B=$Particles.InterpolateOverTime.InterpolateOverTime_Float.Moving Average)
    Scale Color=&RemapRange(
     Input Value To Remap=&FloatFromCurve(
      FloatCurve=DataInterface() {propertyValues {
        bOptimizeLUT=false
        Curve {
         postInfinityExtrap=RCCE_Cycle
         preInfinityExtrap=RCCE_Cycle
         keys [
          <at 0 1 RCIM_Cubic>
          {arriveTangent=0.5283, interpMode=RCIM_Cubic, leaveTangent=0.5283, time=0.0746, value=0.875}
          {arriveTangent=-0.0227, interpMode=RCIM_Cubic, leaveTangent=-0.0227, tangentMode=RCTM_User, time=0.1682, value=1.0888}
          {arriveTangent=-1.3938, interpMode=RCIM_Cubic, leaveTangent=-1.3938, time=0.2019, value=0.6577}
          {arriveTangent=1.4729, interpMode=RCIM_Cubic, leaveTangent=1.4729, time=0.2748, value=0.9402}
          {arriveTangent=0.905, interpMode=RCIM_Cubic, leaveTangent=0.905, time=0.3464, value=0.8705}
          {arriveTangent=-1.2159, interpMode=RCIM_Cubic, leaveTangent=-1.2159, time=0.4074, value=1.0603}
          {arriveTangent=-1.5422, interpMode=RCIM_Cubic, leaveTangent=-1.5422, time=0.4365, value=0.7609}
          {arriveTangent=-0.8529, interpMode=RCIM_Cubic, leaveTangent=-0.8529, time=0.4871, value=0.9373}
          {arriveTangent=0.3027, interpMode=RCIM_Cubic, leaveTangent=0.3027, time=0.5331, value=0.6785}
          {arriveTangent=1.4034, interpMode=RCIM_Cubic, leaveTangent=1.4034, time=0.6112, value=0.9748}
          {arriveTangent=1.5913, interpMode=RCIM_Cubic, leaveTangent=1.5913, time=0.69, value=0.8986}
          {arriveTangent=-0.751, interpMode=RCIM_Cubic, leaveTangent=-0.751, time=0.7536, value=1.2014}
          {arriveTangent=-2.595, interpMode=RCIM_Cubic, leaveTangent=-2.595, time=0.7995, value=0.8164}
          {arriveTangent=-0.8671, interpMode=RCIM_Cubic, leaveTangent=-0.8671, time=0.8674, value=0.9059}
          {arriveTangent=-0.3671, interpMode=RCIM_Cubic, leaveTangent=-0.3671, time=0.9138, value=0.7173}
          {arriveTangent=-0.6714, interpMode=RCIM_Cubic, leaveTangent=-0.6714, time=0.9899, value=0.8609}
          {arriveTangent=0.7195, interpMode=RCIM_Cubic, leaveTangent=0.7195, time=1.0372, value=0.6345}
          {arriveTangent=1.5649, interpMode=RCIM_Cubic, leaveTangent=1.5649, time=1.099, value=0.9394}
          {arriveTangent=0.3682, interpMode=RCIM_Cubic, leaveTangent=0.3682, time=1.1296, value=0.7792}
          {arriveTangent=1.199, interpMode=RCIM_Cubic, leaveTangent=1.199, time=1.2078, value=0.9795}
          {arriveTangent=0.6999, interpMode=RCIM_Cubic, leaveTangent=0.6999, time=1.2281, value=0.8972}
          {arriveTangent=0.8013, interpMode=RCIM_Cubic, leaveTangent=0.8013, time=1.329, value=1.0643}
          {arriveTangent=0.4649, interpMode=RCIM_Cubic, leaveTangent=0.4649, time=1.3719, value=1.0125}
          {arriveTangent=-3.0415, interpMode=RCIM_Cubic, leaveTangent=-3.0415, time=1.414, value=1.1038}
          {arriveTangent=-0.0928, interpMode=RCIM_Cubic, leaveTangent=-0.0928, tangentMode=RCTM_User, time=1.4501, value=0.7746}
          {arriveTangent=0.7682, interpMode=RCIM_Cubic, leaveTangent=0.7682, time=1.5242, value=0.9899}
          {arriveTangent=0.5284, interpMode=RCIM_Cubic, leaveTangent=0.5284, time=1.5566, value=0.8565}
          {arriveTangent=-0.1056, interpMode=RCIM_Cubic, leaveTangent=-0.1056, time=1.7244, value=1.0957}
          {arriveTangent=-2.1193, interpMode=RCIM_Cubic, leaveTangent=-2.1193, time=1.7619, value=0.8348}
          {arriveTangent=-1.2854, interpMode=RCIM_Cubic, leaveTangent=-1.2854, time=1.7959, value=0.9442}
          {arriveTangent=-0.1279, interpMode=RCIM_Cubic, leaveTangent=-0.1279, time=1.836, value=0.7395}
          {arriveTangent=1.0187, interpMode=RCIM_Cubic, leaveTangent=1.0187, time=1.9343, value=0.9265}
          {arriveTangent=1.0782, interpMode=RCIM_Cubic, leaveTangent=1.0782, time=1.9912, value=0.8977}
          {arriveTangent=-2.5847, interpMode=RCIM_Cubic, leaveTangent=-2.5847, time=2.0488, value=1.0499}
          {arriveTangent=-0.6379, interpMode=RCIM_Cubic, leaveTangent=-0.6379, time=2.0828, value=0.6611}
          {arriveTangent=0.4353, interpMode=RCIM_Cubic, leaveTangent=0.4353, time=2.2158, value=0.9434}
          {arriveTangent=-0.3447, interpMode=RCIM_Cubic, leaveTangent=-0.3447, time=2.2673, value=0.7414}
          {arriveTangent=0.6603, interpMode=RCIM_Cubic, leaveTangent=0.6603, time=2.3258, value=0.9055}
          {arriveTangent=0.3801, interpMode=RCIM_Cubic, leaveTangent=0.3801, time=2.3971, value=0.8271}
          {arriveTangent=0.6687, interpMode=RCIM_Cubic, leaveTangent=0.6687, time=2.4654, value=0.9586}
          {arriveTangent=1.2527, interpMode=RCIM_Cubic, leaveTangent=1.2527, time=2.5143, value=0.9055}
          {arriveTangent=-0.515, interpMode=RCIM_Cubic, leaveTangent=-0.515, time=2.5566, value=1.0728}
          {arriveTangent=-1.6201, interpMode=RCIM_Cubic, leaveTangent=-1.6201, time=2.597, value=0.8629}
          {arriveTangent=-0.9669, interpMode=RCIM_Cubic, leaveTangent=-0.9669, time=2.6456, value=0.9286}
          {arriveTangent=0.1009, interpMode=RCIM_Cubic, leaveTangent=0.1009, time=2.6921, value=0.771}
          {arriveTangent=0.0929, interpMode=RCIM_Cubic, leaveTangent=0.0929, time=2.7509, value=0.9392}
          {arriveTangent=0.2939, interpMode=RCIM_Cubic, leaveTangent=0.2939, time=2.8654, value=0.7871}
          {arriveTangent=1.7227, interpMode=RCIM_Cubic, leaveTangent=1.7227, time=2.9271, value=0.991}
          {arriveTangent=0.1236, interpMode=RCIM_Cubic, leaveTangent=0.1236, time=2.957, value=0.9449}
          <at 3 1 RCIM_Cubic>
         ]
        }
       }}
      CurveIndex=$Emitter.NormalizedLoopAge
      Scale Curve=0.8
     )
     Input Min=&Add_Float(A=$System.FlickerCurveRemapMin, B=0)
     Input Max=1.2
     Output Min=0
     Output Max=1
    )
   )
   &FN_Gen5_Light(
    Attenutation:NiagaraFloat=&Multiply_Float(A=15, B=$Particles.InterpolateOverTime.InterpolateOverTime_Float.Moving Average)
    Color:LinearColor=<rgba 1 0.6 0.2 1>
    Exponent=1.5
    Intensity:NiagaraFloat=&Multiply_Float(
     A=&Multiply_Float(A=75, B=1)
     B=&RemapRange(
      Input Value To Remap=&FloatFromCurve(
       FloatCurve=DataInterface() {propertyValues {
         bOptimizeLUT=false
         Curve {
          postInfinityExtrap=RCCE_Cycle
          preInfinityExtrap=RCCE_Cycle
          keys [
           <at 0 1 RCIM_Cubic>
           {arriveTangent=0.5283, interpMode=RCIM_Cubic, leaveTangent=0.5283, time=0.0746, value=0.875}
           {arriveTangent=-0.0227, interpMode=RCIM_Cubic, leaveTangent=-0.0227, tangentMode=RCTM_User, time=0.1682, value=1.0888}
           {arriveTangent=-1.3938, interpMode=RCIM_Cubic, leaveTangent=-1.3938, time=0.2019, value=0.6577}
           {arriveTangent=1.4729, interpMode=RCIM_Cubic, leaveTangent=1.4729, time=0.2748, value=0.9402}
           {arriveTangent=0.905, interpMode=RCIM_Cubic, leaveTangent=0.905, time=0.3464, value=0.8705}
           {arriveTangent=-1.2159, interpMode=RCIM_Cubic, leaveTangent=-1.2159, time=0.4074, value=1.0603}
           {arriveTangent=-1.5422, interpMode=RCIM_Cubic, leaveTangent=-1.5422, time=0.4365, value=0.7609}
           {arriveTangent=-0.8529, interpMode=RCIM_Cubic, leaveTangent=-0.8529, time=0.4871, value=0.9373}
           {arriveTangent=0.3027, interpMode=RCIM_Cubic, leaveTangent=0.3027, time=0.5331, value=0.6785}
           {arriveTangent=1.4034, interpMode=RCIM_Cubic, leaveTangent=1.4034, time=0.6112, value=0.9748}
           {arriveTangent=1.5913, interpMode=RCIM_Cubic, leaveTangent=1.5913, time=0.69, value=0.8986}
           {arriveTangent=-0.751, interpMode=RCIM_Cubic, leaveTangent=-0.751, time=0.7536, value=1.2014}
           {arriveTangent=-2.595, interpMode=RCIM_Cubic, leaveTangent=-2.595, time=0.7995, value=0.8164}
           {arriveTangent=-0.8671, interpMode=RCIM_Cubic, leaveTangent=-0.8671, time=0.8674, value=0.9059}
           {arriveTangent=-0.3671, interpMode=RCIM_Cubic, leaveTangent=-0.3671, time=0.9138, value=0.7173}
           {arriveTangent=-0.6714, interpMode=RCIM_Cubic, leaveTangent=-0.6714, time=0.9899, value=0.8609}
           {arriveTangent=0.7195, interpMode=RCIM_Cubic, leaveTangent=0.7195, time=1.0372, value=0.6345}
           {arriveTangent=1.5649, interpMode=RCIM_Cubic, leaveTangent=1.5649, time=1.099, value=0.9394}
           {arriveTangent=0.3682, interpMode=RCIM_Cubic, leaveTangent=0.3682, time=1.1296, value=0.7792}
           {arriveTangent=1.199, interpMode=RCIM_Cubic, leaveTangent=1.199, time=1.2078, value=0.9795}
           {arriveTangent=0.6999, interpMode=RCIM_Cubic, leaveTangent=0.6999, time=1.2281, value=0.8972}
           {arriveTangent=0.8013, interpMode=RCIM_Cubic, leaveTangent=0.8013, time=1.329, value=1.0643}
           {arriveTangent=0.4649, interpMode=RCIM_Cubic, leaveTangent=0.4649, time=1.3719, value=1.0125}
           {arriveTangent=-3.0415, interpMode=RCIM_Cubic, leaveTangent=-3.0415, time=1.414, value=1.1038}
           {arriveTangent=-0.0928, interpMode=RCIM_Cubic, leaveTangent=-0.0928, tangentMode=RCTM_User, time=1.4501, value=0.7746}
           {arriveTangent=0.7682, interpMode=RCIM_Cubic, leaveTangent=0.7682, time=1.5242, value=0.9899}
           {arriveTangent=0.5284, interpMode=RCIM_Cubic, leaveTangent=0.5284, time=1.5566, value=0.8565}
           {arriveTangent=-0.1056, interpMode=RCIM_Cubic, leaveTangent=-0.1056, time=1.7244, value=1.0957}
           {arriveTangent=-2.1193, interpMode=RCIM_Cubic, leaveTangent=-2.1193, time=1.7619, value=0.8348}
           {arriveTangent=-1.2854, interpMode=RCIM_Cubic, leaveTangent=-1.2854, time=1.7959, value=0.9442}
           {arriveTangent=-0.1279, interpMode=RCIM_Cubic, leaveTangent=-0.1279, time=1.836, value=0.7395}
           {arriveTangent=1.0187, interpMode=RCIM_Cubic, leaveTangent=1.0187, time=1.9343, value=0.9265}
           {arriveTangent=1.0782, interpMode=RCIM_Cubic, leaveTangent=1.0782, time=1.9912, value=0.8977}
           {arriveTangent=-2.5847, interpMode=RCIM_Cubic, leaveTangent=-2.5847, time=2.0488, value=1.0499}
           {arriveTangent=-0.6379, interpMode=RCIM_Cubic, leaveTangent=-0.6379, time=2.0828, value=0.6611}
           {arriveTangent=0.4353, interpMode=RCIM_Cubic, leaveTangent=0.4353, time=2.2158, value=0.9434}
           {arriveTangent=-0.3447, interpMode=RCIM_Cubic, leaveTangent=-0.3447, time=2.2673, value=0.7414}
           {arriveTangent=0.6603, interpMode=RCIM_Cubic, leaveTangent=0.6603, time=2.3258, value=0.9055}
           {arriveTangent=0.3801, interpMode=RCIM_Cubic, leaveTangent=0.3801, time=2.3971, value=0.8271}
           {arriveTangent=0.6687, interpMode=RCIM_Cubic, leaveTangent=0.6687, time=2.4654, value=0.9586}
           {arriveTangent=1.2527, interpMode=RCIM_Cubic, leaveTangent=1.2527, time=2.5143, value=0.9055}
           {arriveTangent=-0.515, interpMode=RCIM_Cubic, leaveTangent=-0.515, time=2.5566, value=1.0728}
           {arriveTangent=-1.6201, interpMode=RCIM_Cubic, leaveTangent=-1.6201, time=2.597, value=0.8629}
           {arriveTangent=-0.9669, interpMode=RCIM_Cubic, leaveTangent=-0.9669, time=2.6456, value=0.9286}
           {arriveTangent=0.1009, interpMode=RCIM_Cubic, leaveTangent=0.1009, time=2.6921, value=0.771}
           {arriveTangent=0.0929, interpMode=RCIM_Cubic, leaveTangent=0.0929, time=2.7509, value=0.9392}
           {arriveTangent=0.2939, interpMode=RCIM_Cubic, leaveTangent=0.2939, time=2.8654, value=0.7871}
           {arriveTangent=1.7227, interpMode=RCIM_Cubic, leaveTangent=1.7227, time=2.9271, value=0.991}
           {arriveTangent=0.1236, interpMode=RCIM_Cubic, leaveTangent=0.1236, time=2.957, value=0.9449}
           <at 3 1 RCIM_Cubic>
          ]
         }
        }}
       CurveIndex=$Emitter.NormalizedLoopAge
       Scale Curve=0.8
      )
      Input Min=&Add_Float(A=$System.FlickerCurveRemapMin, B=0)
      Input Max=1.2
      Output Min=0
      Output Max=1
     )
    )
   )
   /SetVariables(Particles.ComponentsEnabled:NiagaraBool=/Niagara/DynamicInputs/Helpers/MaskBoolBySpawnGroup(Masked Bool:NiagaraBool=false, Passthrough Bool:NiagaraBool=true, Spawn Group Mask:NiagaraInt32=0)) {isSetParameters=true}
   /SetVariables(Particles.LightPos:NiagaraPosition=$Particles.FN_Sprite_Attributes.Position) {isSetParameters=true}
  ]
  rendererClasses [
   &NiagaraSpriteRendererProperties
   &NiagaraLightRendererProperties
   &NiagaraComponentRendererProperties
  ]
  renderers [
   {rendererClass=&NiagaraSpriteRendererProperties}
   {rendererClass=&NiagaraSpriteRendererProperties, rendererIndex=1}
   {rendererClass=&NiagaraLightRendererProperties, rendererIndex=2}
   {rendererClass=&NiagaraComponentRendererProperties, rendererIndex=3}
  ]
 }
 Embers002 {
  emitterUpdateScript [
   &EmitterState(Life Cycle Mode=::NewEnumerator1, Loop Duration=$System.CustomLoopDuration, UseLoopDelay=false)
   &SpawnBurst_Instantaneous(Spawn Group=1, Use Spawn Probability=false, UseLoopCountLimit=false) {module=SpawnBurst_Instantaneous001}
   &SpawnRate(
    Spawn Probability=&RemapRange(
     Input Value To Remap=&FloatFromCurve(
      FloatCurve=DataInterface() {propertyValues {Curve {
         postInfinityExtrap=RCCE_Cycle
         preInfinityExtrap=RCCE_Cycle
         keys [
          <at 0 1 RCIM_Cubic>
          {arriveTangent=0.5283, interpMode=RCIM_Cubic, leaveTangent=0.5283, time=0.0746, value=0.875}
          {arriveTangent=-0.0227, interpMode=RCIM_Cubic, leaveTangent=-0.0227, tangentMode=RCTM_User, time=0.1682, value=1.0888}
          {arriveTangent=-1.3938, interpMode=RCIM_Cubic, leaveTangent=-1.3938, time=0.2019, value=0.6577}
          {arriveTangent=1.4729, interpMode=RCIM_Cubic, leaveTangent=1.4729, time=0.2748, value=0.9402}
          {arriveTangent=0.905, interpMode=RCIM_Cubic, leaveTangent=0.905, time=0.3464, value=0.8705}
          {arriveTangent=-1.2159, interpMode=RCIM_Cubic, leaveTangent=-1.2159, time=0.4074, value=1.0603}
          {arriveTangent=-1.5422, interpMode=RCIM_Cubic, leaveTangent=-1.5422, time=0.4365, value=0.7609}
          {arriveTangent=-0.8529, interpMode=RCIM_Cubic, leaveTangent=-0.8529, time=0.4871, value=0.9373}
          {arriveTangent=0.3027, interpMode=RCIM_Cubic, leaveTangent=0.3027, time=0.5331, value=0.6785}
          {arriveTangent=1.4034, interpMode=RCIM_Cubic, leaveTangent=1.4034, time=0.6112, value=0.9748}
          {arriveTangent=1.5913, interpMode=RCIM_Cubic, leaveTangent=1.5913, time=0.69, value=0.8986}
          {arriveTangent=-0.751, interpMode=RCIM_Cubic, leaveTangent=-0.751, time=0.7536, value=1.2014}
          {arriveTangent=-2.595, interpMode=RCIM_Cubic, leaveTangent=-2.595, time=0.7995, value=0.8164}
          {arriveTangent=-0.8671, interpMode=RCIM_Cubic, leaveTangent=-0.8671, time=0.8674, value=0.9059}
          {arriveTangent=-0.3671, interpMode=RCIM_Cubic, leaveTangent=-0.3671, time=0.9138, value=0.7173}
          {arriveTangent=-0.6714, interpMode=RCIM_Cubic, leaveTangent=-0.6714, time=0.9899, value=0.8609}
          {arriveTangent=0.7195, interpMode=RCIM_Cubic, leaveTangent=0.7195, time=1.0372, value=0.6345}
          {arriveTangent=1.5649, interpMode=RCIM_Cubic, leaveTangent=1.5649, time=1.099, value=0.9394}
          {arriveTangent=0.3682, interpMode=RCIM_Cubic, leaveTangent=0.3682, time=1.1296, value=0.7792}
          {arriveTangent=1.199, interpMode=RCIM_Cubic, leaveTangent=1.199, time=1.2078, value=0.9795}
          {arriveTangent=0.6999, interpMode=RCIM_Cubic, leaveTangent=0.6999, time=1.2281, value=0.8972}
          {arriveTangent=0.8013, interpMode=RCIM_Cubic, leaveTangent=0.8013, time=1.329, value=1.0643}
          {arriveTangent=0.4649, interpMode=RCIM_Cubic, leaveTangent=0.4649, time=1.3719, value=1.0125}
          {arriveTangent=-3.0415, interpMode=RCIM_Cubic, leaveTangent=-3.0415, time=1.414, value=1.1038}
          {arriveTangent=-0.0928, interpMode=RCIM_Cubic, leaveTangent=-0.0928, tangentMode=RCTM_User, time=1.4501, value=0.7746}
          {arriveTangent=0.7682, interpMode=RCIM_Cubic, leaveTangent=0.7682, time=1.5242, value=0.9899}
          {arriveTangent=0.5284, interpMode=RCIM_Cubic, leaveTangent=0.5284, time=1.5566, value=0.8565}
          {arriveTangent=-0.1056, interpMode=RCIM_Cubic, leaveTangent=-0.1056, time=1.7244, value=1.0957}
          {arriveTangent=-2.1193, interpMode=RCIM_Cubic, leaveTangent=-2.1193, time=1.7619, value=0.8348}
          {arriveTangent=-1.2854, interpMode=RCIM_Cubic, leaveTangent=-1.2854, time=1.7959, value=0.9442}
          {arriveTangent=-0.1279, interpMode=RCIM_Cubic, leaveTangent=-0.1279, time=1.836, value=0.7395}
          {arriveTangent=1.0187, interpMode=RCIM_Cubic, leaveTangent=1.0187, time=1.9343, value=0.9265}
          {arriveTangent=1.0782, interpMode=RCIM_Cubic, leaveTangent=1.0782, time=1.9912, value=0.8977}
          {arriveTangent=-2.5847, interpMode=RCIM_Cubic, leaveTangent=-2.5847, time=2.0488, value=1.0499}
          {arriveTangent=-0.6379, interpMode=RCIM_Cubic, leaveTangent=-0.6379, time=2.0828, value=0.6611}
          {arriveTangent=0.4353, interpMode=RCIM_Cubic, leaveTangent=0.4353, time=2.2158, value=0.9434}
          {arriveTangent=-0.3447, interpMode=RCIM_Cubic, leaveTangent=-0.3447, time=2.2673, value=0.7414}
          {arriveTangent=0.6603, interpMode=RCIM_Cubic, leaveTangent=0.6603, time=2.3258, value=0.9055}
          {arriveTangent=0.3801, interpMode=RCIM_Cubic, leaveTangent=0.3801, time=2.3971, value=0.8271}
          {arriveTangent=0.6687, interpMode=RCIM_Cubic, leaveTangent=0.6687, time=2.4654, value=0.9586}
          {arriveTangent=1.2527, interpMode=RCIM_Cubic, leaveTangent=1.2527, time=2.5143, value=0.9055}
          {arriveTangent=-0.515, interpMode=RCIM_Cubic, leaveTangent=-0.515, time=2.5566, value=1.0728}
          {arriveTangent=-1.6201, interpMode=RCIM_Cubic, leaveTangent=-1.6201, time=2.597, value=0.8629}
          {arriveTangent=-0.9669, interpMode=RCIM_Cubic, leaveTangent=-0.9669, time=2.6456, value=0.9286}
          {arriveTangent=0.1009, interpMode=RCIM_Cubic, leaveTangent=0.1009, time=2.6921, value=0.771}
          {arriveTangent=0.0929, interpMode=RCIM_Cubic, leaveTangent=0.0929, time=2.7509, value=0.9392}
          {arriveTangent=0.2939, interpMode=RCIM_Cubic, leaveTangent=0.2939, time=2.8654, value=0.7871}
          {arriveTangent=1.7227, interpMode=RCIM_Cubic, leaveTangent=1.7227, time=2.9271, value=0.991}
          {arriveTangent=0.1236, interpMode=RCIM_Cubic, leaveTangent=0.1236, time=2.957, value=0.9449}
          <at 3 1 RCIM_Cubic>
         ]
        }}}
      CurveIndex=&Add_Float(A=$Emitter.NormalizedLoopAge, B=0)
     )
     Input Min=$System.FlickerCurveRemapMin
     Input Max=1.2
     Output Min=0
     Output Max=0.5
    )
    SpawnRate=&Multiply_Float(A=1, B=100)
    Use Spawn Probability=true
   )
  ]
  particleSpawnScript [
   /SetVariables(Particles.SpawnGroup=$Emitter.SpawnGroup) {isSetParameters=true, module=SetVariables_#3}
   /SetVariables(Particles.SparkIntensity:NiagaraFloat=&FloatFromCurve(
     FloatCurve=DataInterface() {propertyValues {Curve {keys [
         <at 0 1 RCIM_Cubic>
         {arriveTangent=-2.5317, interpMode=RCIM_Cubic, leaveTangent=-2.5317, tangentMode=RCTM_User, time=0.6, value=0.5}
         <at 1 0 RCIM_Cubic>
        ]}}}
     CurveIndex=&RandomRangeFloat(Minimum=0, Maximum=1)
    )) {isSetParameters=true, module=SetVariables_#4}
   &InitializeParticle(
    Color=&MultiplyLinearColor(
     Linear Color=&LerpLinearColors(StartColor=<rgba 10 1.5225 0.5208 1>, EndColor=<rgba 25 16.8856 2.9948 1>, LerpFactor=$Particles.SparkIntensity)
     Scale Factor=<1 0.5 0.1 1>
    )
    Lifetime=&MaskFloatBySpawnGroup(
     Masked Float=&RandomRangeFloat(Minimum=0.17, Maximum=0.25)
     Passthrough Float=$System.CustomLoopDuration
    )
    Sprite Size Mode=::NewEnumerator1
    Uniform Sprite Size Max=5
    Uniform Sprite Size Min=6
    UsePositionOffset=false
    Write Material Random=true
   )
   &SkeletalMeshLocation(
    MaskSpawnGroup=false
    Mesh Sampling Type=::NewEnumerator5
    Skeletal Mesh=DataInterface() {propertyValues {
      DefaultMesh=null
      ExcludeBoneName=root
      OptionalComponentClass=null
      PreviewMesh=&F_MED_NutmegMayo_Pack
      SkinningMode=SkinOnTheFly
      SoftSourceActor=null
      SourceMode=Default
      UvSetIndex=0
      WholeMeshLOD=-1
      bExcludeBone=false
      bReadDeformedGeometry=true
      bRequireCurrentFrameData=true
      ComponentTags []
      FilteredBones []
      FilteredSockets [FXSocket_02]
      MeshUserParameter {parameter {name=None, typeDefHandle {registeredTypeIndex=1}, varData []}}
      SamplingRegions []
     }}
    Socket Sampling Mode=::NewEnumerator5
   )
   &ShapeLocation(
    MaskSpawnGroup=false
    Offset Coordinate Space=::Local
    Offset Mode=::NewEnumerator0
    Override Seed=false
    Sphere Radius=0
    UseOwnerScale=true
   )
   &AddVelocity(
    Module.Use Falloff Distance=false
    Origin Offset Coordinate Space=::World
    Override Seed=false
    Velocity Mode=::NewEnumerator1
    Velocity Origin=$Particles.SkeletalMeshLocation.SampledPosition
    Velocity Speed=&Multiply_Float(A=&RandomRangeFloat(Minimum=20, Maximum=80), B=1)
   ) {module=AddVelocity001}
   &AddVelocity(
    Cone Angle=80
    Cone Axis=<0 0 -1>
    Module.UseConeFalloff=false
    Module.UseVelDistribution=false
    Override Seed=false
    Velocity Mode=::NewEnumerator2
    Velocity Speed=&RandomRangeFloat(Minimum=10, Maximum=80)
   ) {module=AddVelocity002}
   &ApplyInitialForces()
   &FN_Gen5_Light(
    Attenutation=15
    Color:LinearColor=<rgba 1 0.6 0.2 1>
    Intensity:NiagaraFloat=&Multiply_Float(
     A=&Multiply_Float(
      A=150
      B=&MakeCustomFloatFromBool(
       Bool=&SetBoolByIntComparison(A=$Engine.Emitter.NumParticles, B=3, Comparison Type=::NewEnumerator0)
       True Float=1
       False Float=0
      )
     )
     B=&RemapRange(
      Input Value To Remap=&FloatFromCurve(
       FloatCurve=DataInterface() {propertyValues {
         bOptimizeLUT=false
         Curve {
          postInfinityExtrap=RCCE_Cycle
          preInfinityExtrap=RCCE_Cycle
          keys [
           <at 0 1 RCIM_Cubic>
           {arriveTangent=0.5283, interpMode=RCIM_Cubic, leaveTangent=0.5283, time=0.0746, value=0.875}
           {arriveTangent=-0.0227, interpMode=RCIM_Cubic, leaveTangent=-0.0227, tangentMode=RCTM_User, time=0.1682, value=1.0888}
           {arriveTangent=-1.3938, interpMode=RCIM_Cubic, leaveTangent=-1.3938, time=0.2019, value=0.6577}
           {arriveTangent=1.4729, interpMode=RCIM_Cubic, leaveTangent=1.4729, time=0.2748, value=0.9402}
           {arriveTangent=0.905, interpMode=RCIM_Cubic, leaveTangent=0.905, time=0.3464, value=0.8705}
           {arriveTangent=-1.2159, interpMode=RCIM_Cubic, leaveTangent=-1.2159, time=0.4074, value=1.0603}
           {arriveTangent=-1.5422, interpMode=RCIM_Cubic, leaveTangent=-1.5422, time=0.4365, value=0.7609}
           {arriveTangent=-0.8529, interpMode=RCIM_Cubic, leaveTangent=-0.8529, time=0.4871, value=0.9373}
           {arriveTangent=0.3027, interpMode=RCIM_Cubic, leaveTangent=0.3027, time=0.5331, value=0.6785}
           {arriveTangent=1.4034, interpMode=RCIM_Cubic, leaveTangent=1.4034, time=0.6112, value=0.9748}
           {arriveTangent=1.5913, interpMode=RCIM_Cubic, leaveTangent=1.5913, time=0.69, value=0.8986}
           {arriveTangent=-0.751, interpMode=RCIM_Cubic, leaveTangent=-0.751, time=0.7536, value=1.2014}
           {arriveTangent=-2.595, interpMode=RCIM_Cubic, leaveTangent=-2.595, time=0.7995, value=0.8164}
           {arriveTangent=-0.8671, interpMode=RCIM_Cubic, leaveTangent=-0.8671, time=0.8674, value=0.9059}
           {arriveTangent=-0.3671, interpMode=RCIM_Cubic, leaveTangent=-0.3671, time=0.9138, value=0.7173}
           {arriveTangent=-0.6714, interpMode=RCIM_Cubic, leaveTangent=-0.6714, time=0.9899, value=0.8609}
           {arriveTangent=0.7195, interpMode=RCIM_Cubic, leaveTangent=0.7195, time=1.0372, value=0.6345}
           {arriveTangent=1.5649, interpMode=RCIM_Cubic, leaveTangent=1.5649, time=1.099, value=0.9394}
           {arriveTangent=0.3682, interpMode=RCIM_Cubic, leaveTangent=0.3682, time=1.1296, value=0.7792}
           {arriveTangent=1.199, interpMode=RCIM_Cubic, leaveTangent=1.199, time=1.2078, value=0.9795}
           {arriveTangent=0.6999, interpMode=RCIM_Cubic, leaveTangent=0.6999, time=1.2281, value=0.8972}
           {arriveTangent=0.8013, interpMode=RCIM_Cubic, leaveTangent=0.8013, time=1.329, value=1.0643}
           {arriveTangent=0.4649, interpMode=RCIM_Cubic, leaveTangent=0.4649, time=1.3719, value=1.0125}
           {arriveTangent=-3.0415, interpMode=RCIM_Cubic, leaveTangent=-3.0415, time=1.414, value=1.1038}
           {arriveTangent=-0.0928, interpMode=RCIM_Cubic, leaveTangent=-0.0928, tangentMode=RCTM_User, time=1.4501, value=0.7746}
           {arriveTangent=0.7682, interpMode=RCIM_Cubic, leaveTangent=0.7682, time=1.5242, value=0.9899}
           {arriveTangent=0.5284, interpMode=RCIM_Cubic, leaveTangent=0.5284, time=1.5566, value=0.8565}
           {arriveTangent=-0.1056, interpMode=RCIM_Cubic, leaveTangent=-0.1056, time=1.7244, value=1.0957}
           {arriveTangent=-2.1193, interpMode=RCIM_Cubic, leaveTangent=-2.1193, time=1.7619, value=0.8348}
           {arriveTangent=-1.2854, interpMode=RCIM_Cubic, leaveTangent=-1.2854, time=1.7959, value=0.9442}
           {arriveTangent=-0.1279, interpMode=RCIM_Cubic, leaveTangent=-0.1279, time=1.836, value=0.7395}
           {arriveTangent=1.0187, interpMode=RCIM_Cubic, leaveTangent=1.0187, time=1.9343, value=0.9265}
           {arriveTangent=1.0782, interpMode=RCIM_Cubic, leaveTangent=1.0782, time=1.9912, value=0.8977}
           {arriveTangent=-2.5847, interpMode=RCIM_Cubic, leaveTangent=-2.5847, time=2.0488, value=1.0499}
           {arriveTangent=-0.6379, interpMode=RCIM_Cubic, leaveTangent=-0.6379, time=2.0828, value=0.6611}
           {arriveTangent=0.4353, interpMode=RCIM_Cubic, leaveTangent=0.4353, time=2.2158, value=0.9434}
           {arriveTangent=-0.3447, interpMode=RCIM_Cubic, leaveTangent=-0.3447, time=2.2673, value=0.7414}
           {arriveTangent=0.6603, interpMode=RCIM_Cubic, leaveTangent=0.6603, time=2.3258, value=0.9055}
           {arriveTangent=0.3801, interpMode=RCIM_Cubic, leaveTangent=0.3801, time=2.3971, value=0.8271}
           {arriveTangent=0.6687, interpMode=RCIM_Cubic, leaveTangent=0.6687, time=2.4654, value=0.9586}
           {arriveTangent=1.2527, interpMode=RCIM_Cubic, leaveTangent=1.2527, time=2.5143, value=0.9055}
           {arriveTangent=-0.515, interpMode=RCIM_Cubic, leaveTangent=-0.515, time=2.5566, value=1.0728}
           {arriveTangent=-1.6201, interpMode=RCIM_Cubic, leaveTangent=-1.6201, time=2.597, value=0.8629}
           {arriveTangent=-0.9669, interpMode=RCIM_Cubic, leaveTangent=-0.9669, time=2.6456, value=0.9286}
           {arriveTangent=0.1009, interpMode=RCIM_Cubic, leaveTangent=0.1009, time=2.6921, value=0.771}
           {arriveTangent=0.0929, interpMode=RCIM_Cubic, leaveTangent=0.0929, time=2.7509, value=0.9392}
           {arriveTangent=0.2939, interpMode=RCIM_Cubic, leaveTangent=0.2939, time=2.8654, value=0.7871}
           {arriveTangent=1.7227, interpMode=RCIM_Cubic, leaveTangent=1.7227, time=2.9271, value=0.991}
           {arriveTangent=0.1236, interpMode=RCIM_Cubic, leaveTangent=0.1236, time=2.957, value=0.9449}
           <at 3 1 RCIM_Cubic>
          ]
         }
        }}
       CurveIndex=$Emitter.NormalizedLoopAge
      )
      Input Min=&Add_Float(A=$System.FlickerCurveRemapMin, B=0)
      Input Max=1.2
      Output Min=0
      Output Max=1
     )
    )
   ) {module=FN_Gen5_Light001}
  ]
  particleUpdateScript [
   &ParticleState()
   &ScaleSpriteSize(Scale Factor=&Multiply_Vector2D_ByFloat(Float=0.1, Vector2D=<1 0.7>), Scale Sprite Size Mode=::NewEnumerator2)
   &SkeletalMeshLocation(
    MaskSpawnGroup=false
    Mesh Sampling Type=::NewEnumerator5
    Position Sampling=::NewEnumerator2
    Skeletal Mesh=DataInterface() {propertyValues {
      DefaultMesh=null
      ExcludeBoneName=root
      OptionalComponentClass=null
      PreviewMesh=&F_MED_NutmegMayo_Pack
      SkinningMode=SkinOnTheFly
      SoftSourceActor=null
      SourceMode=Default
      UvSetIndex=0
      WholeMeshLOD=-1
      bExcludeBone=false
      bReadDeformedGeometry=true
      bRequireCurrentFrameData=true
      ComponentTags []
      FilteredBones []
      FilteredSockets [FXSocket_02]
      MeshUserParameter {parameter {name=None, typeDefHandle {registeredTypeIndex=1}, varData []}}
      SamplingRegions []
     }}
    Socket Sampling Mode=::NewEnumerator5
   ) {module=SkeletalMeshLocation001}
   &FN_Sprite_Attributes(
    Camera Offset=15
    Color=&MultiplyLinearColor(
     Linear Color=$Particles.Color
     Scale Factor=&MakeVector4(
      X=1
      Y=1
      Z=1
      W=&Multiply_Float(
       A=0.1
       B=&FloatFromCurve(FloatCurve=DataInterface() {propertyValues {Curve {keys [
            {interpMode=RCIM_Cubic, tangentMode=RCTM_User, time=0, value=0}
            {interpMode=RCIM_Cubic, tangentMode=RCTM_User, time=0.5, value=1}
            {tangentMode=RCTM_User, time=1, value=0}
           ]}}})
      )
     )
    )
    Create Camera Offset Attribute=true
    Create Color Attribute=true
    Create Position Attribute=true
    Create SpriteSize Attribute=true
    Position:NiagaraPosition=$Particles.SkeletalMeshLocation001.SampledPosition
    SpriteSize=&Multiply_Vector2D_ByFloat(Float=0.05, Vector2D=&Multiply_Vector2D(A=<400 400>, B=$Particles.SpriteSize))
    Write to Module Namespace=true
   )
   &Drag(Use Linear Drag=true, Use Rotational Drag=false)
   &GravityForce(Gravity=&Multiply_Vector(A=<0 0 -980>, B=<1 1 0.35>))
   &InheritSourceMovement(ApplyForce=false, ApplyPosition=true, ApplyPositionFalloff=false)
   &SolveForcesAndVelocity(Clamp Velocity=false, Limit Acceleration=false)
   &ScaleSpriteSizeBySpeed(Max Scale Factor=<1 5>, Min Scale Factor=<1 -5>, Velocity Threshold=50)
   /SetVariables(Particles.FlareDynamic:Vector4f=<0.1 1 0 0>) {isSetParameters=true, module=SetVariables_#5}
   /SetVariables(Particles.VisibilityTag=$Particles.Initial.SpawnGroup) {isSetParameters=true, module=SetVariables_#6}
   &InterpolateOverTime(
    Rate Of Change=5
    Target Float Value:NiagaraFloat=&MakeCustomFloatFromBool(
     Bool=&SetBoolByIntComparison(A=$Engine.Emitter.NumParticles, B=2, Comparison Type=::NewEnumerator0)
     True Float=1
     False Float=0
    )
    Use Set Interpolation Rate=true
   )
   &FN_Light_Attributes(
    Color=$Particles.Color
    Create Color Attribute=true
    Create Exponent Attribute=true
    Create Position Attribute=true
    Create Radius Attribute=true
    Enable Particle Light Enable=true
    Exponent=1.5
    Light Enabled=true
    Position=$Particles.FN_Sprite_Attributes.Position
    Radius=&Multiply_Float(A=15, B=$Particles.InterpolateOverTime.InterpolateOverTime_Float.Moving Average)
    Scale Color=&RemapRange(
     Input Value To Remap=&FloatFromCurve(
      FloatCurve=DataInterface() {propertyValues {
        bOptimizeLUT=false
        Curve {
         postInfinityExtrap=RCCE_Cycle
         preInfinityExtrap=RCCE_Cycle
         keys [
          <at 0 1 RCIM_Cubic>
          {arriveTangent=0.5283, interpMode=RCIM_Cubic, leaveTangent=0.5283, time=0.0746, value=0.875}
          {arriveTangent=-0.0227, interpMode=RCIM_Cubic, leaveTangent=-0.0227, tangentMode=RCTM_User, time=0.1682, value=1.0888}
          {arriveTangent=-1.3938, interpMode=RCIM_Cubic, leaveTangent=-1.3938, time=0.2019, value=0.6577}
          {arriveTangent=1.4729, interpMode=RCIM_Cubic, leaveTangent=1.4729, time=0.2748, value=0.9402}
          {arriveTangent=0.905, interpMode=RCIM_Cubic, leaveTangent=0.905, time=0.3464, value=0.8705}
          {arriveTangent=-1.2159, interpMode=RCIM_Cubic, leaveTangent=-1.2159, time=0.4074, value=1.0603}
          {arriveTangent=-1.5422, interpMode=RCIM_Cubic, leaveTangent=-1.5422, time=0.4365, value=0.7609}
          {arriveTangent=-0.8529, interpMode=RCIM_Cubic, leaveTangent=-0.8529, time=0.4871, value=0.9373}
          {arriveTangent=0.3027, interpMode=RCIM_Cubic, leaveTangent=0.3027, time=0.5331, value=0.6785}
          {arriveTangent=1.4034, interpMode=RCIM_Cubic, leaveTangent=1.4034, time=0.6112, value=0.9748}
          {arriveTangent=1.5913, interpMode=RCIM_Cubic, leaveTangent=1.5913, time=0.69, value=0.8986}
          {arriveTangent=-0.751, interpMode=RCIM_Cubic, leaveTangent=-0.751, time=0.7536, value=1.2014}
          {arriveTangent=-2.595, interpMode=RCIM_Cubic, leaveTangent=-2.595, time=0.7995, value=0.8164}
          {arriveTangent=-0.8671, interpMode=RCIM_Cubic, leaveTangent=-0.8671, time=0.8674, value=0.9059}
          {arriveTangent=-0.3671, interpMode=RCIM_Cubic, leaveTangent=-0.3671, time=0.9138, value=0.7173}
          {arriveTangent=-0.6714, interpMode=RCIM_Cubic, leaveTangent=-0.6714, time=0.9899, value=0.8609}
          {arriveTangent=0.7195, interpMode=RCIM_Cubic, leaveTangent=0.7195, time=1.0372, value=0.6345}
          {arriveTangent=1.5649, interpMode=RCIM_Cubic, leaveTangent=1.5649, time=1.099, value=0.9394}
          {arriveTangent=0.3682, interpMode=RCIM_Cubic, leaveTangent=0.3682, time=1.1296, value=0.7792}
          {arriveTangent=1.199, interpMode=RCIM_Cubic, leaveTangent=1.199, time=1.2078, value=0.9795}
          {arriveTangent=0.6999, interpMode=RCIM_Cubic, leaveTangent=0.6999, time=1.2281, value=0.8972}
          {arriveTangent=0.8013, interpMode=RCIM_Cubic, leaveTangent=0.8013, time=1.329, value=1.0643}
          {arriveTangent=0.4649, interpMode=RCIM_Cubic, leaveTangent=0.4649, time=1.3719, value=1.0125}
          {arriveTangent=-3.0415, interpMode=RCIM_Cubic, leaveTangent=-3.0415, time=1.414, value=1.1038}
          {arriveTangent=-0.0928, interpMode=RCIM_Cubic, leaveTangent=-0.0928, tangentMode=RCTM_User, time=1.4501, value=0.7746}
          {arriveTangent=0.7682, interpMode=RCIM_Cubic, leaveTangent=0.7682, time=1.5242, value=0.9899}
          {arriveTangent=0.5284, interpMode=RCIM_Cubic, leaveTangent=0.5284, time=1.5566, value=0.8565}
          {arriveTangent=-0.1056, interpMode=RCIM_Cubic, leaveTangent=-0.1056, time=1.7244, value=1.0957}
          {arriveTangent=-2.1193, interpMode=RCIM_Cubic, leaveTangent=-2.1193, time=1.7619, value=0.8348}
          {arriveTangent=-1.2854, interpMode=RCIM_Cubic, leaveTangent=-1.2854, time=1.7959, value=0.9442}
          {arriveTangent=-0.1279, interpMode=RCIM_Cubic, leaveTangent=-0.1279, time=1.836, value=0.7395}
          {arriveTangent=1.0187, interpMode=RCIM_Cubic, leaveTangent=1.0187, time=1.9343, value=0.9265}
          {arriveTangent=1.0782, interpMode=RCIM_Cubic, leaveTangent=1.0782, time=1.9912, value=0.8977}
          {arriveTangent=-2.5847, interpMode=RCIM_Cubic, leaveTangent=-2.5847, time=2.0488, value=1.0499}
          {arriveTangent=-0.6379, interpMode=RCIM_Cubic, leaveTangent=-0.6379, time=2.0828, value=0.6611}
          {arriveTangent=0.4353, interpMode=RCIM_Cubic, leaveTangent=0.4353, time=2.2158, value=0.9434}
          {arriveTangent=-0.3447, interpMode=RCIM_Cubic, leaveTangent=-0.3447, time=2.2673, value=0.7414}
          {arriveTangent=0.6603, interpMode=RCIM_Cubic, leaveTangent=0.6603, time=2.3258, value=0.9055}
          {arriveTangent=0.3801, interpMode=RCIM_Cubic, leaveTangent=0.3801, time=2.3971, value=0.8271}
          {arriveTangent=0.6687, interpMode=RCIM_Cubic, leaveTangent=0.6687, time=2.4654, value=0.9586}
          {arriveTangent=1.2527, interpMode=RCIM_Cubic, leaveTangent=1.2527, time=2.5143, value=0.9055}
          {arriveTangent=-0.515, interpMode=RCIM_Cubic, leaveTangent=-0.515, time=2.5566, value=1.0728}
          {arriveTangent=-1.6201, interpMode=RCIM_Cubic, leaveTangent=-1.6201, time=2.597, value=0.8629}
          {arriveTangent=-0.9669, interpMode=RCIM_Cubic, leaveTangent=-0.9669, time=2.6456, value=0.9286}
          {arriveTangent=0.1009, interpMode=RCIM_Cubic, leaveTangent=0.1009, time=2.6921, value=0.771}
          {arriveTangent=0.0929, interpMode=RCIM_Cubic, leaveTangent=0.0929, time=2.7509, value=0.9392}
          {arriveTangent=0.2939, interpMode=RCIM_Cubic, leaveTangent=0.2939, time=2.8654, value=0.7871}
          {arriveTangent=1.7227, interpMode=RCIM_Cubic, leaveTangent=1.7227, time=2.9271, value=0.991}
          {arriveTangent=0.1236, interpMode=RCIM_Cubic, leaveTangent=0.1236, time=2.957, value=0.9449}
          <at 3 1 RCIM_Cubic>
         ]
        }
       }}
      CurveIndex=$Emitter.NormalizedLoopAge
      Scale Curve=0.8
     )
     Input Min=&Add_Float(A=$System.FlickerCurveRemapMin, B=0)
     Input Max=1.2
     Output Min=0
     Output Max=1
    )
   )
   &FN_Gen5_Light(
    Attenutation:NiagaraFloat=&Multiply_Float(A=15, B=$Particles.InterpolateOverTime.InterpolateOverTime_Float.Moving Average)
    Color:LinearColor=<rgba 1 0.6 0.2 1>
    Exponent=1.5
    Intensity:NiagaraFloat=&Multiply_Float(
     A=&Multiply_Float(A=75, B=1)
     B=&RemapRange(
      Input Value To Remap=&FloatFromCurve(
       FloatCurve=DataInterface() {propertyValues {
         bOptimizeLUT=false
         Curve {
          postInfinityExtrap=RCCE_Cycle
          preInfinityExtrap=RCCE_Cycle
          keys [
           <at 0 1 RCIM_Cubic>
           {arriveTangent=0.5283, interpMode=RCIM_Cubic, leaveTangent=0.5283, time=0.0746, value=0.875}
           {arriveTangent=-0.0227, interpMode=RCIM_Cubic, leaveTangent=-0.0227, tangentMode=RCTM_User, time=0.1682, value=1.0888}
           {arriveTangent=-1.3938, interpMode=RCIM_Cubic, leaveTangent=-1.3938, time=0.2019, value=0.6577}
           {arriveTangent=1.4729, interpMode=RCIM_Cubic, leaveTangent=1.4729, time=0.2748, value=0.9402}
           {arriveTangent=0.905, interpMode=RCIM_Cubic, leaveTangent=0.905, time=0.3464, value=0.8705}
           {arriveTangent=-1.2159, interpMode=RCIM_Cubic, leaveTangent=-1.2159, time=0.4074, value=1.0603}
           {arriveTangent=-1.5422, interpMode=RCIM_Cubic, leaveTangent=-1.5422, time=0.4365, value=0.7609}
           {arriveTangent=-0.8529, interpMode=RCIM_Cubic, leaveTangent=-0.8529, time=0.4871, value=0.9373}
           {arriveTangent=0.3027, interpMode=RCIM_Cubic, leaveTangent=0.3027, time=0.5331, value=0.6785}
           {arriveTangent=1.4034, interpMode=RCIM_Cubic, leaveTangent=1.4034, time=0.6112, value=0.9748}
           {arriveTangent=1.5913, interpMode=RCIM_Cubic, leaveTangent=1.5913, time=0.69, value=0.8986}
           {arriveTangent=-0.751, interpMode=RCIM_Cubic, leaveTangent=-0.751, time=0.7536, value=1.2014}
           {arriveTangent=-2.595, interpMode=RCIM_Cubic, leaveTangent=-2.595, time=0.7995, value=0.8164}
           {arriveTangent=-0.8671, interpMode=RCIM_Cubic, leaveTangent=-0.8671, time=0.8674, value=0.9059}
           {arriveTangent=-0.3671, interpMode=RCIM_Cubic, leaveTangent=-0.3671, time=0.9138, value=0.7173}
           {arriveTangent=-0.6714, interpMode=RCIM_Cubic, leaveTangent=-0.6714, time=0.9899, value=0.8609}
           {arriveTangent=0.7195, interpMode=RCIM_Cubic, leaveTangent=0.7195, time=1.0372, value=0.6345}
           {arriveTangent=1.5649, interpMode=RCIM_Cubic, leaveTangent=1.5649, time=1.099, value=0.9394}
           {arriveTangent=0.3682, interpMode=RCIM_Cubic, leaveTangent=0.3682, time=1.1296, value=0.7792}
           {arriveTangent=1.199, interpMode=RCIM_Cubic, leaveTangent=1.199, time=1.2078, value=0.9795}
           {arriveTangent=0.6999, interpMode=RCIM_Cubic, leaveTangent=0.6999, time=1.2281, value=0.8972}
           {arriveTangent=0.8013, interpMode=RCIM_Cubic, leaveTangent=0.8013, time=1.329, value=1.0643}
           {arriveTangent=0.4649, interpMode=RCIM_Cubic, leaveTangent=0.4649, time=1.3719, value=1.0125}
           {arriveTangent=-3.0415, interpMode=RCIM_Cubic, leaveTangent=-3.0415, time=1.414, value=1.1038}
           {arriveTangent=-0.0928, interpMode=RCIM_Cubic, leaveTangent=-0.0928, tangentMode=RCTM_User, time=1.4501, value=0.7746}
           {arriveTangent=0.7682, interpMode=RCIM_Cubic, leaveTangent=0.7682, time=1.5242, value=0.9899}
           {arriveTangent=0.5284, interpMode=RCIM_Cubic, leaveTangent=0.5284, time=1.5566, value=0.8565}
           {arriveTangent=-0.1056, interpMode=RCIM_Cubic, leaveTangent=-0.1056, time=1.7244, value=1.0957}
           {arriveTangent=-2.1193, interpMode=RCIM_Cubic, leaveTangent=-2.1193, time=1.7619, value=0.8348}
           {arriveTangent=-1.2854, interpMode=RCIM_Cubic, leaveTangent=-1.2854, time=1.7959, value=0.9442}
           {arriveTangent=-0.1279, interpMode=RCIM_Cubic, leaveTangent=-0.1279, time=1.836, value=0.7395}
           {arriveTangent=1.0187, interpMode=RCIM_Cubic, leaveTangent=1.0187, time=1.9343, value=0.9265}
           {arriveTangent=1.0782, interpMode=RCIM_Cubic, leaveTangent=1.0782, time=1.9912, value=0.8977}
           {arriveTangent=-2.5847, interpMode=RCIM_Cubic, leaveTangent=-2.5847, time=2.0488, value=1.0499}
           {arriveTangent=-0.6379, interpMode=RCIM_Cubic, leaveTangent=-0.6379, time=2.0828, value=0.6611}
           {arriveTangent=0.4353, interpMode=RCIM_Cubic, leaveTangent=0.4353, time=2.2158, value=0.9434}
           {arriveTangent=-0.3447, interpMode=RCIM_Cubic, leaveTangent=-0.3447, time=2.2673, value=0.7414}
           {arriveTangent=0.6603, interpMode=RCIM_Cubic, leaveTangent=0.6603, time=2.3258, value=0.9055}
           {arriveTangent=0.3801, interpMode=RCIM_Cubic, leaveTangent=0.3801, time=2.3971, value=0.8271}
           {arriveTangent=0.6687, interpMode=RCIM_Cubic, leaveTangent=0.6687, time=2.4654, value=0.9586}
           {arriveTangent=1.2527, interpMode=RCIM_Cubic, leaveTangent=1.2527, time=2.5143, value=0.9055}
           {arriveTangent=-0.515, interpMode=RCIM_Cubic, leaveTangent=-0.515, time=2.5566, value=1.0728}
           {arriveTangent=-1.6201, interpMode=RCIM_Cubic, leaveTangent=-1.6201, time=2.597, value=0.8629}
           {arriveTangent=-0.9669, interpMode=RCIM_Cubic, leaveTangent=-0.9669, time=2.6456, value=0.9286}
           {arriveTangent=0.1009, interpMode=RCIM_Cubic, leaveTangent=0.1009, time=2.6921, value=0.771}
           {arriveTangent=0.0929, interpMode=RCIM_Cubic, leaveTangent=0.0929, time=2.7509, value=0.9392}
           {arriveTangent=0.2939, interpMode=RCIM_Cubic, leaveTangent=0.2939, time=2.8654, value=0.7871}
           {arriveTangent=1.7227, interpMode=RCIM_Cubic, leaveTangent=1.7227, time=2.9271, value=0.991}
           {arriveTangent=0.1236, interpMode=RCIM_Cubic, leaveTangent=0.1236, time=2.957, value=0.9449}
           <at 3 1 RCIM_Cubic>
          ]
         }
        }}
       CurveIndex=$Emitter.NormalizedLoopAge
       Scale Curve=0.8
      )
      Input Min=&Add_Float(A=$System.FlickerCurveRemapMin, B=0)
      Input Max=1.2
      Output Min=0
      Output Max=1
     )
    )
   )
   /SetVariables(Particles.ComponentsEnabled:NiagaraBool=&MaskBoolBySpawnGroup(Masked Bool:NiagaraBool=false, Passthrough Bool:NiagaraBool=true, Spawn Group Mask:NiagaraInt32=0)) {isSetParameters=true}
   /SetVariables(Particles.LightPos:NiagaraPosition=$Particles.FN_Sprite_Attributes.Position) {isSetParameters=true}
  ]
  rendererClasses [
   &NiagaraSpriteRendererProperties
   &NiagaraLightRendererProperties
   &NiagaraComponentRendererProperties
  ]
  renderers [
   {rendererClass=&NiagaraSpriteRendererProperties}
   {rendererClass=&NiagaraSpriteRendererProperties, rendererIndex=1}
   {rendererClass=&NiagaraLightRendererProperties, rendererIndex=2}
   {rendererClass=&NiagaraComponentRendererProperties, rendererIndex=3}
  ]
 }
 Embers003 {
  emitterUpdateScript [
   &EmitterState(Life Cycle Mode=::NewEnumerator1, Loop Duration=$System.CustomLoopDuration, UseLoopDelay=false)
   &SpawnBurst_Instantaneous(Spawn Group=1, Use Spawn Probability=false, UseLoopCountLimit=false) {module=SpawnBurst_Instantaneous001}
   &SpawnRate(
    Spawn Probability=&RemapRange(
     Input Value To Remap=&FloatFromCurve(
      FloatCurve=DataInterface() {propertyValues {Curve {
         postInfinityExtrap=RCCE_Cycle
         preInfinityExtrap=RCCE_Cycle
         keys [
          <at 0 1 RCIM_Cubic>
          {arriveTangent=0.5283, interpMode=RCIM_Cubic, leaveTangent=0.5283, time=0.0746, value=0.875}
          {arriveTangent=-0.0227, interpMode=RCIM_Cubic, leaveTangent=-0.0227, tangentMode=RCTM_User, time=0.1682, value=1.0888}
          {arriveTangent=-1.3938, interpMode=RCIM_Cubic, leaveTangent=-1.3938, time=0.2019, value=0.6577}
          {arriveTangent=1.4729, interpMode=RCIM_Cubic, leaveTangent=1.4729, time=0.2748, value=0.9402}
          {arriveTangent=0.905, interpMode=RCIM_Cubic, leaveTangent=0.905, time=0.3464, value=0.8705}
          {arriveTangent=-1.2159, interpMode=RCIM_Cubic, leaveTangent=-1.2159, time=0.4074, value=1.0603}
          {arriveTangent=-1.5422, interpMode=RCIM_Cubic, leaveTangent=-1.5422, time=0.4365, value=0.7609}
          {arriveTangent=-0.8529, interpMode=RCIM_Cubic, leaveTangent=-0.8529, time=0.4871, value=0.9373}
          {arriveTangent=0.3027, interpMode=RCIM_Cubic, leaveTangent=0.3027, time=0.5331, value=0.6785}
          {arriveTangent=1.4034, interpMode=RCIM_Cubic, leaveTangent=1.4034, time=0.6112, value=0.9748}
          {arriveTangent=1.5913, interpMode=RCIM_Cubic, leaveTangent=1.5913, time=0.69, value=0.8986}
          {arriveTangent=-0.751, interpMode=RCIM_Cubic, leaveTangent=-0.751, time=0.7536, value=1.2014}
          {arriveTangent=-2.595, interpMode=RCIM_Cubic, leaveTangent=-2.595, time=0.7995, value=0.8164}
          {arriveTangent=-0.8671, interpMode=RCIM_Cubic, leaveTangent=-0.8671, time=0.8674, value=0.9059}
          {arriveTangent=-0.3671, interpMode=RCIM_Cubic, leaveTangent=-0.3671, time=0.9138, value=0.7173}
          {arriveTangent=-0.6714, interpMode=RCIM_Cubic, leaveTangent=-0.6714, time=0.9899, value=0.8609}
          {arriveTangent=0.7195, interpMode=RCIM_Cubic, leaveTangent=0.7195, time=1.0372, value=0.6345}
          {arriveTangent=1.5649, interpMode=RCIM_Cubic, leaveTangent=1.5649, time=1.099, value=0.9394}
          {arriveTangent=0.3682, interpMode=RCIM_Cubic, leaveTangent=0.3682, time=1.1296, value=0.7792}
          {arriveTangent=1.199, interpMode=RCIM_Cubic, leaveTangent=1.199, time=1.2078, value=0.9795}
          {arriveTangent=0.6999, interpMode=RCIM_Cubic, leaveTangent=0.6999, time=1.2281, value=0.8972}
          {arriveTangent=0.8013, interpMode=RCIM_Cubic, leaveTangent=0.8013, time=1.329, value=1.0643}
          {arriveTangent=0.4649, interpMode=RCIM_Cubic, leaveTangent=0.4649, time=1.3719, value=1.0125}
          {arriveTangent=-3.0415, interpMode=RCIM_Cubic, leaveTangent=-3.0415, time=1.414, value=1.1038}
          {arriveTangent=-0.0928, interpMode=RCIM_Cubic, leaveTangent=-0.0928, tangentMode=RCTM_User, time=1.4501, value=0.7746}
          {arriveTangent=0.7682, interpMode=RCIM_Cubic, leaveTangent=0.7682, time=1.5242, value=0.9899}
          {arriveTangent=0.5284, interpMode=RCIM_Cubic, leaveTangent=0.5284, time=1.5566, value=0.8565}
          {arriveTangent=-0.1056, interpMode=RCIM_Cubic, leaveTangent=-0.1056, time=1.7244, value=1.0957}
          {arriveTangent=-2.1193, interpMode=RCIM_Cubic, leaveTangent=-2.1193, time=1.7619, value=0.8348}
          {arriveTangent=-1.2854, interpMode=RCIM_Cubic, leaveTangent=-1.2854, time=1.7959, value=0.9442}
          {arriveTangent=-0.1279, interpMode=RCIM_Cubic, leaveTangent=-0.1279, time=1.836, value=0.7395}
          {arriveTangent=1.0187, interpMode=RCIM_Cubic, leaveTangent=1.0187, time=1.9343, value=0.9265}
          {arriveTangent=1.0782, interpMode=RCIM_Cubic, leaveTangent=1.0782, time=1.9912, value=0.8977}
          {arriveTangent=-2.5847, interpMode=RCIM_Cubic, leaveTangent=-2.5847, time=2.0488, value=1.0499}
          {arriveTangent=-0.6379, interpMode=RCIM_Cubic, leaveTangent=-0.6379, time=2.0828, value=0.6611}
          {arriveTangent=0.4353, interpMode=RCIM_Cubic, leaveTangent=0.4353, time=2.2158, value=0.9434}
          {arriveTangent=-0.3447, interpMode=RCIM_Cubic, leaveTangent=-0.3447, time=2.2673, value=0.7414}
          {arriveTangent=0.6603, interpMode=RCIM_Cubic, leaveTangent=0.6603, time=2.3258, value=0.9055}
          {arriveTangent=0.3801, interpMode=RCIM_Cubic, leaveTangent=0.3801, time=2.3971, value=0.8271}
          {arriveTangent=0.6687, interpMode=RCIM_Cubic, leaveTangent=0.6687, time=2.4654, value=0.9586}
          {arriveTangent=1.2527, interpMode=RCIM_Cubic, leaveTangent=1.2527, time=2.5143, value=0.9055}
          {arriveTangent=-0.515, interpMode=RCIM_Cubic, leaveTangent=-0.515, time=2.5566, value=1.0728}
          {arriveTangent=-1.6201, interpMode=RCIM_Cubic, leaveTangent=-1.6201, time=2.597, value=0.8629}
          {arriveTangent=-0.9669, interpMode=RCIM_Cubic, leaveTangent=-0.9669, time=2.6456, value=0.9286}
          {arriveTangent=0.1009, interpMode=RCIM_Cubic, leaveTangent=0.1009, time=2.6921, value=0.771}
          {arriveTangent=0.0929, interpMode=RCIM_Cubic, leaveTangent=0.0929, time=2.7509, value=0.9392}
          {arriveTangent=0.2939, interpMode=RCIM_Cubic, leaveTangent=0.2939, time=2.8654, value=0.7871}
          {arriveTangent=1.7227, interpMode=RCIM_Cubic, leaveTangent=1.7227, time=2.9271, value=0.991}
          {arriveTangent=0.1236, interpMode=RCIM_Cubic, leaveTangent=0.1236, time=2.957, value=0.9449}
          <at 3 1 RCIM_Cubic>
         ]
        }}}
      CurveIndex=&Add_Float(A=$Emitter.NormalizedLoopAge, B=0)
     )
     Input Min=$System.FlickerCurveRemapMin
     Input Max=1.2
     Output Min=0
     Output Max=0.5
    )
    SpawnRate=&Multiply_Float(A=1, B=100)
    Use Spawn Probability=true
   )
  ]
  particleSpawnScript [
   /SetVariables(Particles.SpawnGroup=$Emitter.SpawnGroup) {isSetParameters=true, module=SetVariables_#3}
   /SetVariables(Particles.SparkIntensity:NiagaraFloat=&FloatFromCurve(
     FloatCurve=DataInterface() {propertyValues {Curve {keys [
         <at 0 1 RCIM_Cubic>
         {arriveTangent=-2.5317, interpMode=RCIM_Cubic, leaveTangent=-2.5317, tangentMode=RCTM_User, time=0.6, value=0.5}
         <at 1 0 RCIM_Cubic>
        ]}}}
     CurveIndex=&RandomRangeFloat(Minimum=0, Maximum=1)
    )) {isSetParameters=true, module=SetVariables_#4}
   &InitializeParticle(
    Color=&MultiplyLinearColor(
     Linear Color=&LerpLinearColors(StartColor=<rgba 10 1.5225 0.5208 1>, EndColor=<rgba 25 16.8856 2.9948 1>, LerpFactor=$Particles.SparkIntensity)
     Scale Factor=<1 0.5 0.1 1>
    )
    Lifetime=&MaskFloatBySpawnGroup(
     Masked Float=&RandomRangeFloat(Minimum=0.17, Maximum=0.25)
     Passthrough Float=$System.CustomLoopDuration
    )
    Sprite Size Mode=::NewEnumerator1
    Uniform Sprite Size Max=5
    Uniform Sprite Size Min=6
    UsePositionOffset=false
    Write Material Random=true
   )
   &SkeletalMeshLocation(
    MaskSpawnGroup=false
    Mesh Sampling Type=::NewEnumerator5
    Skeletal Mesh=DataInterface() {propertyValues {
      DefaultMesh=null
      ExcludeBoneName=root
      OptionalComponentClass=null
      PreviewMesh=&F_MED_NutmegMayo_Pack
      SkinningMode=SkinOnTheFly
      SoftSourceActor=null
      SourceMode=Default
      UvSetIndex=0
      WholeMeshLOD=-1
      bExcludeBone=false
      bReadDeformedGeometry=true
      bRequireCurrentFrameData=true
      ComponentTags []
      FilteredBones []
      FilteredSockets [FXSocket_03]
      MeshUserParameter {parameter {name=None, typeDefHandle {registeredTypeIndex=1}, varData []}}
      SamplingRegions []
     }}
    Socket Sampling Mode=::NewEnumerator5
   )
   &ShapeLocation(
    MaskSpawnGroup=false
    Offset Coordinate Space=::Local
    Offset Mode=::NewEnumerator0
    Override Seed=false
    Sphere Radius=0
    UseOwnerScale=true
   )
   &AddVelocity(
    Module.Use Falloff Distance=false
    Origin Offset Coordinate Space=::World
    Override Seed=false
    Velocity Mode=::NewEnumerator1
    Velocity Origin=$Particles.SkeletalMeshLocation.SampledPosition
    Velocity Speed=&Multiply_Float(A=&RandomRangeFloat(Minimum=20, Maximum=80), B=1)
   ) {module=AddVelocity001}
   &AddVelocity(
    Cone Angle=80
    Cone Axis=<0.1 -2 1>
    Module.UseConeFalloff=false
    Module.UseVelDistribution=false
    Override Seed=false
    Velocity Mode=::NewEnumerator2
    Velocity Speed=&RandomRangeFloat(Minimum=10, Maximum=80)
   ) {module=AddVelocity002}
   &ApplyInitialForces()
   &FN_Gen5_Light(
    Attenutation=15
    Color:LinearColor=<rgba 1 0.6 0.2 1>
    Intensity:NiagaraFloat=&Multiply_Float(
     A=&Multiply_Float(
      A=150
      B=&MakeCustomFloatFromBool(
       Bool=&SetBoolByIntComparison(A=$Engine.Emitter.NumParticles, B=3, Comparison Type=::NewEnumerator0)
       True Float=1
       False Float=0
      )
     )
     B=&RemapRange(
      Input Value To Remap=&FloatFromCurve(
       FloatCurve=DataInterface() {propertyValues {
         bOptimizeLUT=false
         Curve {
          postInfinityExtrap=RCCE_Cycle
          preInfinityExtrap=RCCE_Cycle
          keys [
           <at 0 1 RCIM_Cubic>
           {arriveTangent=0.5283, interpMode=RCIM_Cubic, leaveTangent=0.5283, time=0.0746, value=0.875}
           {arriveTangent=-0.0227, interpMode=RCIM_Cubic, leaveTangent=-0.0227, tangentMode=RCTM_User, time=0.1682, value=1.0888}
           {arriveTangent=-1.3938, interpMode=RCIM_Cubic, leaveTangent=-1.3938, time=0.2019, value=0.6577}
           {arriveTangent=1.4729, interpMode=RCIM_Cubic, leaveTangent=1.4729, time=0.2748, value=0.9402}
           {arriveTangent=0.905, interpMode=RCIM_Cubic, leaveTangent=0.905, time=0.3464, value=0.8705}
           {arriveTangent=-1.2159, interpMode=RCIM_Cubic, leaveTangent=-1.2159, time=0.4074, value=1.0603}
           {arriveTangent=-1.5422, interpMode=RCIM_Cubic, leaveTangent=-1.5422, time=0.4365, value=0.7609}
           {arriveTangent=-0.8529, interpMode=RCIM_Cubic, leaveTangent=-0.8529, time=0.4871, value=0.9373}
           {arriveTangent=0.3027, interpMode=RCIM_Cubic, leaveTangent=0.3027, time=0.5331, value=0.6785}
           {arriveTangent=1.4034, interpMode=RCIM_Cubic, leaveTangent=1.4034, time=0.6112, value=0.9748}
           {arriveTangent=1.5913, interpMode=RCIM_Cubic, leaveTangent=1.5913, time=0.69, value=0.8986}
           {arriveTangent=-0.751, interpMode=RCIM_Cubic, leaveTangent=-0.751, time=0.7536, value=1.2014}
           {arriveTangent=-2.595, interpMode=RCIM_Cubic, leaveTangent=-2.595, time=0.7995, value=0.8164}
           {arriveTangent=-0.8671, interpMode=RCIM_Cubic, leaveTangent=-0.8671, time=0.8674, value=0.9059}
           {arriveTangent=-0.3671, interpMode=RCIM_Cubic, leaveTangent=-0.3671, time=0.9138, value=0.7173}
           {arriveTangent=-0.6714, interpMode=RCIM_Cubic, leaveTangent=-0.6714, time=0.9899, value=0.8609}
           {arriveTangent=0.7195, interpMode=RCIM_Cubic, leaveTangent=0.7195, time=1.0372, value=0.6345}
           {arriveTangent=1.5649, interpMode=RCIM_Cubic, leaveTangent=1.5649, time=1.099, value=0.9394}
           {arriveTangent=0.3682, interpMode=RCIM_Cubic, leaveTangent=0.3682, time=1.1296, value=0.7792}
           {arriveTangent=1.199, interpMode=RCIM_Cubic, leaveTangent=1.199, time=1.2078, value=0.9795}
           {arriveTangent=0.6999, interpMode=RCIM_Cubic, leaveTangent=0.6999, time=1.2281, value=0.8972}
           {arriveTangent=0.8013, interpMode=RCIM_Cubic, leaveTangent=0.8013, time=1.329, value=1.0643}
           {arriveTangent=0.4649, interpMode=RCIM_Cubic, leaveTangent=0.4649, time=1.3719, value=1.0125}
           {arriveTangent=-3.0415, interpMode=RCIM_Cubic, leaveTangent=-3.0415, time=1.414, value=1.1038}
           {arriveTangent=-0.0928, interpMode=RCIM_Cubic, leaveTangent=-0.0928, tangentMode=RCTM_User, time=1.4501, value=0.7746}
           {arriveTangent=0.7682, interpMode=RCIM_Cubic, leaveTangent=0.7682, time=1.5242, value=0.9899}
           {arriveTangent=0.5284, interpMode=RCIM_Cubic, leaveTangent=0.5284, time=1.5566, value=0.8565}
           {arriveTangent=-0.1056, interpMode=RCIM_Cubic, leaveTangent=-0.1056, time=1.7244, value=1.0957}
           {arriveTangent=-2.1193, interpMode=RCIM_Cubic, leaveTangent=-2.1193, time=1.7619, value=0.8348}
           {arriveTangent=-1.2854, interpMode=RCIM_Cubic, leaveTangent=-1.2854, time=1.7959, value=0.9442}
           {arriveTangent=-0.1279, interpMode=RCIM_Cubic, leaveTangent=-0.1279, time=1.836, value=0.7395}
           {arriveTangent=1.0187, interpMode=RCIM_Cubic, leaveTangent=1.0187, time=1.9343, value=0.9265}
           {arriveTangent=1.0782, interpMode=RCIM_Cubic, leaveTangent=1.0782, time=1.9912, value=0.8977}
           {arriveTangent=-2.5847, interpMode=RCIM_Cubic, leaveTangent=-2.5847, time=2.0488, value=1.0499}
           {arriveTangent=-0.6379, interpMode=RCIM_Cubic, leaveTangent=-0.6379, time=2.0828, value=0.6611}
           {arriveTangent=0.4353, interpMode=RCIM_Cubic, leaveTangent=0.4353, time=2.2158, value=0.9434}
           {arriveTangent=-0.3447, interpMode=RCIM_Cubic, leaveTangent=-0.3447, time=2.2673, value=0.7414}
           {arriveTangent=0.6603, interpMode=RCIM_Cubic, leaveTangent=0.6603, time=2.3258, value=0.9055}
           {arriveTangent=0.3801, interpMode=RCIM_Cubic, leaveTangent=0.3801, time=2.3971, value=0.8271}
           {arriveTangent=0.6687, interpMode=RCIM_Cubic, leaveTangent=0.6687, time=2.4654, value=0.9586}
           {arriveTangent=1.2527, interpMode=RCIM_Cubic, leaveTangent=1.2527, time=2.5143, value=0.9055}
           {arriveTangent=-0.515, interpMode=RCIM_Cubic, leaveTangent=-0.515, time=2.5566, value=1.0728}
           {arriveTangent=-1.6201, interpMode=RCIM_Cubic, leaveTangent=-1.6201, time=2.597, value=0.8629}
           {arriveTangent=-0.9669, interpMode=RCIM_Cubic, leaveTangent=-0.9669, time=2.6456, value=0.9286}
           {arriveTangent=0.1009, interpMode=RCIM_Cubic, leaveTangent=0.1009, time=2.6921, value=0.771}
           {arriveTangent=0.0929, interpMode=RCIM_Cubic, leaveTangent=0.0929, time=2.7509, value=0.9392}
           {arriveTangent=0.2939, interpMode=RCIM_Cubic, leaveTangent=0.2939, time=2.8654, value=0.7871}
           {arriveTangent=1.7227, interpMode=RCIM_Cubic, leaveTangent=1.7227, time=2.9271, value=0.991}
           {arriveTangent=0.1236, interpMode=RCIM_Cubic, leaveTangent=0.1236, time=2.957, value=0.9449}
           <at 3 1 RCIM_Cubic>
          ]
         }
        }}
       CurveIndex=$Emitter.NormalizedLoopAge
      )
      Input Min=&Add_Float(A=$System.FlickerCurveRemapMin, B=0)
      Input Max=1.2
      Output Min=0
      Output Max=1
     )
    )
   ) {module=FN_Gen5_Light001}
  ]
  particleUpdateScript [
   &ParticleState()
   &ScaleSpriteSize(Scale Factor=&Multiply_Vector2D_ByFloat(Float=0.1, Vector2D=<1 0.7>), Scale Sprite Size Mode=::NewEnumerator2)
   &SkeletalMeshLocation(
    MaskSpawnGroup=false
    Mesh Sampling Type=::NewEnumerator5
    Position Sampling=::NewEnumerator2
    Skeletal Mesh=DataInterface() {propertyValues {
      DefaultMesh=null
      ExcludeBoneName=root
      OptionalComponentClass=null
      PreviewMesh=&F_MED_NutmegMayo_Pack
      SkinningMode=SkinOnTheFly
      SoftSourceActor=null
      SourceMode=Default
      UvSetIndex=0
      WholeMeshLOD=-1
      bExcludeBone=false
      bReadDeformedGeometry=true
      bRequireCurrentFrameData=true
      ComponentTags []
      FilteredBones []
      FilteredSockets [FXSocket_03]
      MeshUserParameter {parameter {name=None, typeDefHandle {registeredTypeIndex=1}, varData []}}
      SamplingRegions []
     }}
    Socket Sampling Mode=::NewEnumerator5
   ) {module=SkeletalMeshLocation001}
   &FN_Sprite_Attributes(
    Camera Offset=1
    Color=&MultiplyLinearColor(
     Linear Color=$Particles.Color
     Scale Factor=&MakeVector4(
      X=1
      Y=1
      Z=1
      W=&Multiply_Float(
       A=0.1
       B=&FloatFromCurve(FloatCurve=DataInterface() {propertyValues {Curve {keys [
            {interpMode=RCIM_Cubic, tangentMode=RCTM_User, time=0, value=0}
            {interpMode=RCIM_Cubic, tangentMode=RCTM_User, time=0.5, value=1}
            {tangentMode=RCTM_User, time=1, value=0}
           ]}}})
      )
     )
    )
    Create Camera Offset Attribute=true
    Create Color Attribute=true
    Create Position Attribute=true
    Create SpriteSize Attribute=true
    Position:NiagaraPosition=$Particles.SkeletalMeshLocation001.SampledPosition
    SpriteSize=&Multiply_Vector2D_ByFloat(Float=0.05, Vector2D=&Multiply_Vector2D(A=<400 400>, B=$Particles.SpriteSize))
    Write to Module Namespace=true
   )
   &Drag(Use Linear Drag=true, Use Rotational Drag=false)
   &GravityForce(Gravity=&Multiply_Vector(A=<0 0 -980>, B=<1 1 0.35>))
   &InheritSourceMovement(ApplyForce=false, ApplyPosition=true, ApplyPositionFalloff=false)
   &SolveForcesAndVelocity(Clamp Velocity=false, Limit Acceleration=false)
   &ScaleSpriteSizeBySpeed(Max Scale Factor=<1 5>, Min Scale Factor=<1 -5>, Velocity Threshold=50)
   /SetVariables(Particles.FlareDynamic:Vector4f=<0.1 1 0 0>) {isSetParameters=true, module=SetVariables_#5}
   /SetVariables(Particles.VisibilityTag=$Particles.Initial.SpawnGroup) {isSetParameters=true, module=SetVariables_#6}
   &InterpolateOverTime(
    Rate Of Change=5
    Target Float Value:NiagaraFloat=&MakeCustomFloatFromBool(
     Bool=&SetBoolByIntComparison(A=$Engine.Emitter.NumParticles, B=2, Comparison Type=::NewEnumerator0)
     True Float=1
     False Float=0
    )
    Use Set Interpolation Rate=true
   )
   &FN_Light_Attributes(
    Color=$Particles.Color
    Create Color Attribute=true
    Create Exponent Attribute=true
    Create Position Attribute=true
    Create Radius Attribute=true
    Enable Particle Light Enable=true
    Exponent=1.5
    Light Enabled=true
    Position=$Particles.FN_Sprite_Attributes.Position
    Radius=&Multiply_Float(
     A=&Multiply_Float(A=15, B=$Particles.InterpolateOverTime.InterpolateOverTime_Float.Moving Average)
     B=&FloatFromCurve(
      FloatCurve=DataInterface() {propertyValues {
        bOptimizeLUT=false
        Curve {
         postInfinityExtrap=RCCE_Cycle
         preInfinityExtrap=RCCE_Cycle
         keys [
          <at 0 1 RCIM_Cubic>
          {arriveTangent=0.5283, interpMode=RCIM_Cubic, leaveTangent=0.5283, time=0.0746, value=0.875}
          {arriveTangent=-0.0227, interpMode=RCIM_Cubic, leaveTangent=-0.0227, tangentMode=RCTM_User, time=0.1682, value=1.0888}
          {arriveTangent=-1.3938, interpMode=RCIM_Cubic, leaveTangent=-1.3938, time=0.2019, value=0.6577}
          {arriveTangent=1.4729, interpMode=RCIM_Cubic, leaveTangent=1.4729, time=0.2748, value=0.9402}
          {arriveTangent=0.905, interpMode=RCIM_Cubic, leaveTangent=0.905, time=0.3464, value=0.8705}
          {arriveTangent=-1.2159, interpMode=RCIM_Cubic, leaveTangent=-1.2159, time=0.4074, value=1.0603}
          {arriveTangent=-1.5422, interpMode=RCIM_Cubic, leaveTangent=-1.5422, time=0.4365, value=0.7609}
          {arriveTangent=-0.8529, interpMode=RCIM_Cubic, leaveTangent=-0.8529, time=0.4871, value=0.9373}
          {arriveTangent=0.3027, interpMode=RCIM_Cubic, leaveTangent=0.3027, time=0.5331, value=0.6785}
          {arriveTangent=1.4034, interpMode=RCIM_Cubic, leaveTangent=1.4034, time=0.6112, value=0.9748}
          {arriveTangent=1.5913, interpMode=RCIM_Cubic, leaveTangent=1.5913, time=0.69, value=0.8986}
          {arriveTangent=-0.751, interpMode=RCIM_Cubic, leaveTangent=-0.751, time=0.7536, value=1.2014}
          {arriveTangent=-2.595, interpMode=RCIM_Cubic, leaveTangent=-2.595, time=0.7995, value=0.8164}
          {arriveTangent=-0.8671, interpMode=RCIM_Cubic, leaveTangent=-0.8671, time=0.8674, value=0.9059}
          {arriveTangent=-0.3671, interpMode=RCIM_Cubic, leaveTangent=-0.3671, time=0.9138, value=0.7173}
          {arriveTangent=-0.6714, interpMode=RCIM_Cubic, leaveTangent=-0.6714, time=0.9899, value=0.8609}
          {arriveTangent=0.7195, interpMode=RCIM_Cubic, leaveTangent=0.7195, time=1.0372, value=0.6345}
          {arriveTangent=1.5649, interpMode=RCIM_Cubic, leaveTangent=1.5649, time=1.099, value=0.9394}
          {arriveTangent=0.3682, interpMode=RCIM_Cubic, leaveTangent=0.3682, time=1.1296, value=0.7792}
          {arriveTangent=1.199, interpMode=RCIM_Cubic, leaveTangent=1.199, time=1.2078, value=0.9795}
          {arriveTangent=0.6999, interpMode=RCIM_Cubic, leaveTangent=0.6999, time=1.2281, value=0.8972}
          {arriveTangent=0.8013, interpMode=RCIM_Cubic, leaveTangent=0.8013, time=1.329, value=1.0643}
          {arriveTangent=0.4649, interpMode=RCIM_Cubic, leaveTangent=0.4649, time=1.3719, value=1.0125}
          {arriveTangent=-3.0415, interpMode=RCIM_Cubic, leaveTangent=-3.0415, time=1.414, value=1.1038}
          {arriveTangent=-0.0928, interpMode=RCIM_Cubic, leaveTangent=-0.0928, tangentMode=RCTM_User, time=1.4501, value=0.7746}
          {arriveTangent=0.7682, interpMode=RCIM_Cubic, leaveTangent=0.7682, time=1.5242, value=0.9899}
          {arriveTangent=0.5284, interpMode=RCIM_Cubic, leaveTangent=0.5284, time=1.5566, value=0.8565}
          {arriveTangent=-0.1056, interpMode=RCIM_Cubic, leaveTangent=-0.1056, time=1.7244, value=1.0957}
          {arriveTangent=-2.1193, interpMode=RCIM_Cubic, leaveTangent=-2.1193, time=1.7619, value=0.8348}
          {arriveTangent=-1.2854, interpMode=RCIM_Cubic, leaveTangent=-1.2854, time=1.7959, value=0.9442}
          {arriveTangent=-0.1279, interpMode=RCIM_Cubic, leaveTangent=-0.1279, time=1.836, value=0.7395}
          {arriveTangent=1.0187, interpMode=RCIM_Cubic, leaveTangent=1.0187, time=1.9343, value=0.9265}
          {arriveTangent=1.0782, interpMode=RCIM_Cubic, leaveTangent=1.0782, time=1.9912, value=0.8977}
          {arriveTangent=-2.5847, interpMode=RCIM_Cubic, leaveTangent=-2.5847, time=2.0488, value=1.0499}
          {arriveTangent=-0.6379, interpMode=RCIM_Cubic, leaveTangent=-0.6379, time=2.0828, value=0.6611}
          {arriveTangent=0.4353, interpMode=RCIM_Cubic, leaveTangent=0.4353, time=2.2158, value=0.9434}
          {arriveTangent=-0.3447, interpMode=RCIM_Cubic, leaveTangent=-0.3447, time=2.2673, value=0.7414}
          {arriveTangent=0.6603, interpMode=RCIM_Cubic, leaveTangent=0.6603, time=2.3258, value=0.9055}
          {arriveTangent=0.3801, interpMode=RCIM_Cubic, leaveTangent=0.3801, time=2.3971, value=0.8271}
          {arriveTangent=0.6687, interpMode=RCIM_Cubic, leaveTangent=0.6687, time=2.4654, value=0.9586}
          {arriveTangent=1.2527, interpMode=RCIM_Cubic, leaveTangent=1.2527, time=2.5143, value=0.9055}
          {arriveTangent=-0.515, interpMode=RCIM_Cubic, leaveTangent=-0.515, time=2.5566, value=1.0728}
          {arriveTangent=-1.6201, interpMode=RCIM_Cubic, leaveTangent=-1.6201, time=2.597, value=0.8629}
          {arriveTangent=-0.9669, interpMode=RCIM_Cubic, leaveTangent=-0.9669, time=2.6456, value=0.9286}
          {arriveTangent=0.1009, interpMode=RCIM_Cubic, leaveTangent=0.1009, time=2.6921, value=0.771}
          {arriveTangent=0.0929, interpMode=RCIM_Cubic, leaveTangent=0.0929, time=2.7509, value=0.9392}
          {arriveTangent=0.2939, interpMode=RCIM_Cubic, leaveTangent=0.2939, time=2.8654, value=0.7871}
          {arriveTangent=1.7227, interpMode=RCIM_Cubic, leaveTangent=1.7227, time=2.9271, value=0.991}
          {arriveTangent=0.1236, interpMode=RCIM_Cubic, leaveTangent=0.1236, time=2.957, value=0.9449}
          <at 3 1 RCIM_Cubic>
         ]
        }
       }}
      CurveIndex=$Engine.Time
     )
    )
    Scale Color=&RemapRange(
     Input Value To Remap=&FloatFromCurve(
      FloatCurve=DataInterface() {propertyValues {
        bOptimizeLUT=false
        Curve {
         postInfinityExtrap=RCCE_Cycle
         preInfinityExtrap=RCCE_Cycle
         keys [
          <at 0 1 RCIM_Cubic>
          {arriveTangent=0.5283, interpMode=RCIM_Cubic, leaveTangent=0.5283, time=0.0746, value=0.875}
          {arriveTangent=-0.0227, interpMode=RCIM_Cubic, leaveTangent=-0.0227, tangentMode=RCTM_User, time=0.1682, value=1.0888}
          {arriveTangent=-1.3938, interpMode=RCIM_Cubic, leaveTangent=-1.3938, time=0.2019, value=0.6577}
          {arriveTangent=1.4729, interpMode=RCIM_Cubic, leaveTangent=1.4729, time=0.2748, value=0.9402}
          {arriveTangent=0.905, interpMode=RCIM_Cubic, leaveTangent=0.905, time=0.3464, value=0.8705}
          {arriveTangent=-1.2159, interpMode=RCIM_Cubic, leaveTangent=-1.2159, time=0.4074, value=1.0603}
          {arriveTangent=-1.5422, interpMode=RCIM_Cubic, leaveTangent=-1.5422, time=0.4365, value=0.7609}
          {arriveTangent=-0.8529, interpMode=RCIM_Cubic, leaveTangent=-0.8529, time=0.4871, value=0.9373}
          {arriveTangent=0.3027, interpMode=RCIM_Cubic, leaveTangent=0.3027, time=0.5331, value=0.6785}
          {arriveTangent=1.4034, interpMode=RCIM_Cubic, leaveTangent=1.4034, time=0.6112, value=0.9748}
          {arriveTangent=1.5913, interpMode=RCIM_Cubic, leaveTangent=1.5913, time=0.69, value=0.8986}
          {arriveTangent=-0.751, interpMode=RCIM_Cubic, leaveTangent=-0.751, time=0.7536, value=1.2014}
          {arriveTangent=-2.595, interpMode=RCIM_Cubic, leaveTangent=-2.595, time=0.7995, value=0.8164}
          {arriveTangent=-0.8671, interpMode=RCIM_Cubic, leaveTangent=-0.8671, time=0.8674, value=0.9059}
          {arriveTangent=-0.3671, interpMode=RCIM_Cubic, leaveTangent=-0.3671, time=0.9138, value=0.7173}
          {arriveTangent=-0.6714, interpMode=RCIM_Cubic, leaveTangent=-0.6714, time=0.9899, value=0.8609}
          {arriveTangent=0.7195, interpMode=RCIM_Cubic, leaveTangent=0.7195, time=1.0372, value=0.6345}
          {arriveTangent=1.5649, interpMode=RCIM_Cubic, leaveTangent=1.5649, time=1.099, value=0.9394}
          {arriveTangent=0.3682, interpMode=RCIM_Cubic, leaveTangent=0.3682, time=1.1296, value=0.7792}
          {arriveTangent=1.199, interpMode=RCIM_Cubic, leaveTangent=1.199, time=1.2078, value=0.9795}
          {arriveTangent=0.6999, interpMode=RCIM_Cubic, leaveTangent=0.6999, time=1.2281, value=0.8972}
          {arriveTangent=0.8013, interpMode=RCIM_Cubic, leaveTangent=0.8013, time=1.329, value=1.0643}
          {arriveTangent=0.4649, interpMode=RCIM_Cubic, leaveTangent=0.4649, time=1.3719, value=1.0125}
          {arriveTangent=-3.0415, interpMode=RCIM_Cubic, leaveTangent=-3.0415, time=1.414, value=1.1038}
          {arriveTangent=-0.0928, interpMode=RCIM_Cubic, leaveTangent=-0.0928, tangentMode=RCTM_User, time=1.4501, value=0.7746}
          {arriveTangent=0.7682, interpMode=RCIM_Cubic, leaveTangent=0.7682, time=1.5242, value=0.9899}
          {arriveTangent=0.5284, interpMode=RCIM_Cubic, leaveTangent=0.5284, time=1.5566, value=0.8565}
          {arriveTangent=-0.1056, interpMode=RCIM_Cubic, leaveTangent=-0.1056, time=1.7244, value=1.0957}
          {arriveTangent=-2.1193, interpMode=RCIM_Cubic, leaveTangent=-2.1193, time=1.7619, value=0.8348}
          {arriveTangent=-1.2854, interpMode=RCIM_Cubic, leaveTangent=-1.2854, time=1.7959, value=0.9442}
          {arriveTangent=-0.1279, interpMode=RCIM_Cubic, leaveTangent=-0.1279, time=1.836, value=0.7395}
          {arriveTangent=1.0187, interpMode=RCIM_Cubic, leaveTangent=1.0187, time=1.9343, value=0.9265}
          {arriveTangent=1.0782, interpMode=RCIM_Cubic, leaveTangent=1.0782, time=1.9912, value=0.8977}
          {arriveTangent=-2.5847, interpMode=RCIM_Cubic, leaveTangent=-2.5847, time=2.0488, value=1.0499}
          {arriveTangent=-0.6379, interpMode=RCIM_Cubic, leaveTangent=-0.6379, time=2.0828, value=0.6611}
          {arriveTangent=0.4353, interpMode=RCIM_Cubic, leaveTangent=0.4353, time=2.2158, value=0.9434}
          {arriveTangent=-0.3447, interpMode=RCIM_Cubic, leaveTangent=-0.3447, time=2.2673, value=0.7414}
          {arriveTangent=0.6603, interpMode=RCIM_Cubic, leaveTangent=0.6603, time=2.3258, value=0.9055}
          {arriveTangent=0.3801, interpMode=RCIM_Cubic, leaveTangent=0.3801, time=2.3971, value=0.8271}
          {arriveTangent=0.6687, interpMode=RCIM_Cubic, leaveTangent=0.6687, time=2.4654, value=0.9586}
          {arriveTangent=1.2527, interpMode=RCIM_Cubic, leaveTangent=1.2527, time=2.5143, value=0.9055}
          {arriveTangent=-0.515, interpMode=RCIM_Cubic, leaveTangent=-0.515, time=2.5566, value=1.0728}
          {arriveTangent=-1.6201, interpMode=RCIM_Cubic, leaveTangent=-1.6201, time=2.597, value=0.8629}
          {arriveTangent=-0.9669, interpMode=RCIM_Cubic, leaveTangent=-0.9669, time=2.6456, value=0.9286}
          {arriveTangent=0.1009, interpMode=RCIM_Cubic, leaveTangent=0.1009, time=2.6921, value=0.771}
          {arriveTangent=0.0929, interpMode=RCIM_Cubic, leaveTangent=0.0929, time=2.7509, value=0.9392}
          {arriveTangent=0.2939, interpMode=RCIM_Cubic, leaveTangent=0.2939, time=2.8654, value=0.7871}
          {arriveTangent=1.7227, interpMode=RCIM_Cubic, leaveTangent=1.7227, time=2.9271, value=0.991}
          {arriveTangent=0.1236, interpMode=RCIM_Cubic, leaveTangent=0.1236, time=2.957, value=0.9449}
          <at 3 1 RCIM_Cubic>
         ]
        }
       }}
      CurveIndex=$Emitter.NormalizedLoopAge
      Scale Curve=0.8
     )
     Input Min=&Add_Float(A=$System.FlickerCurveRemapMin, B=0)
     Input Max=1.2
     Output Min=0
     Output Max=1
    )
   )
   &FN_Gen5_Light(
    Attenutation:NiagaraFloat=&Multiply_Float(A=&Multiply_Float(A=15, B=$Particles.InterpolateOverTime.InterpolateOverTime_Float.Moving Average), B=1)
    Color:LinearColor=<rgba 1 0.6 0.2 1>
    Exponent=1.5
    Intensity:NiagaraFloat=&Multiply_Float(
     A=&Multiply_Float(A=75, B=1)
     B=&RemapRange(
      Input Value To Remap=&FloatFromCurve(
       FloatCurve=DataInterface() {propertyValues {
         bOptimizeLUT=false
         Curve {
          postInfinityExtrap=RCCE_Cycle
          preInfinityExtrap=RCCE_Cycle
          keys [
           <at 0 1 RCIM_Cubic>
           {arriveTangent=0.5283, interpMode=RCIM_Cubic, leaveTangent=0.5283, time=0.0746, value=0.875}
           {arriveTangent=-0.0227, interpMode=RCIM_Cubic, leaveTangent=-0.0227, tangentMode=RCTM_User, time=0.1682, value=1.0888}
           {arriveTangent=-1.3938, interpMode=RCIM_Cubic, leaveTangent=-1.3938, time=0.2019, value=0.6577}
           {arriveTangent=1.4729, interpMode=RCIM_Cubic, leaveTangent=1.4729, time=0.2748, value=0.9402}
           {arriveTangent=0.905, interpMode=RCIM_Cubic, leaveTangent=0.905, time=0.3464, value=0.8705}
           {arriveTangent=-1.2159, interpMode=RCIM_Cubic, leaveTangent=-1.2159, time=0.4074, value=1.0603}
           {arriveTangent=-1.5422, interpMode=RCIM_Cubic, leaveTangent=-1.5422, time=0.4365, value=0.7609}
           {arriveTangent=-0.8529, interpMode=RCIM_Cubic, leaveTangent=-0.8529, time=0.4871, value=0.9373}
           {arriveTangent=0.3027, interpMode=RCIM_Cubic, leaveTangent=0.3027, time=0.5331, value=0.6785}
           {arriveTangent=1.4034, interpMode=RCIM_Cubic, leaveTangent=1.4034, time=0.6112, value=0.9748}
           {arriveTangent=1.5913, interpMode=RCIM_Cubic, leaveTangent=1.5913, time=0.69, value=0.8986}
           {arriveTangent=-0.751, interpMode=RCIM_Cubic, leaveTangent=-0.751, time=0.7536, value=1.2014}
           {arriveTangent=-2.595, interpMode=RCIM_Cubic, leaveTangent=-2.595, time=0.7995, value=0.8164}
           {arriveTangent=-0.8671, interpMode=RCIM_Cubic, leaveTangent=-0.8671, time=0.8674, value=0.9059}
           {arriveTangent=-0.3671, interpMode=RCIM_Cubic, leaveTangent=-0.3671, time=0.9138, value=0.7173}
           {arriveTangent=-0.6714, interpMode=RCIM_Cubic, leaveTangent=-0.6714, time=0.9899, value=0.8609}
           {arriveTangent=0.7195, interpMode=RCIM_Cubic, leaveTangent=0.7195, time=1.0372, value=0.6345}
           {arriveTangent=1.5649, interpMode=RCIM_Cubic, leaveTangent=1.5649, time=1.099, value=0.9394}
           {arriveTangent=0.3682, interpMode=RCIM_Cubic, leaveTangent=0.3682, time=1.1296, value=0.7792}
           {arriveTangent=1.199, interpMode=RCIM_Cubic, leaveTangent=1.199, time=1.2078, value=0.9795}
           {arriveTangent=0.6999, interpMode=RCIM_Cubic, leaveTangent=0.6999, time=1.2281, value=0.8972}
           {arriveTangent=0.8013, interpMode=RCIM_Cubic, leaveTangent=0.8013, time=1.329, value=1.0643}
           {arriveTangent=0.4649, interpMode=RCIM_Cubic, leaveTangent=0.4649, time=1.3719, value=1.0125}
           {arriveTangent=-3.0415, interpMode=RCIM_Cubic, leaveTangent=-3.0415, time=1.414, value=1.1038}
           {arriveTangent=-0.0928, interpMode=RCIM_Cubic, leaveTangent=-0.0928, tangentMode=RCTM_User, time=1.4501, value=0.7746}
           {arriveTangent=0.7682, interpMode=RCIM_Cubic, leaveTangent=0.7682, time=1.5242, value=0.9899}
           {arriveTangent=0.5284, interpMode=RCIM_Cubic, leaveTangent=0.5284, time=1.5566, value=0.8565}
           {arriveTangent=-0.1056, interpMode=RCIM_Cubic, leaveTangent=-0.1056, time=1.7244, value=1.0957}
           {arriveTangent=-2.1193, interpMode=RCIM_Cubic, leaveTangent=-2.1193, time=1.7619, value=0.8348}
           {arriveTangent=-1.2854, interpMode=RCIM_Cubic, leaveTangent=-1.2854, time=1.7959, value=0.9442}
           {arriveTangent=-0.1279, interpMode=RCIM_Cubic, leaveTangent=-0.1279, time=1.836, value=0.7395}
           {arriveTangent=1.0187, interpMode=RCIM_Cubic, leaveTangent=1.0187, time=1.9343, value=0.9265}
           {arriveTangent=1.0782, interpMode=RCIM_Cubic, leaveTangent=1.0782, time=1.9912, value=0.8977}
           {arriveTangent=-2.5847, interpMode=RCIM_Cubic, leaveTangent=-2.5847, time=2.0488, value=1.0499}
           {arriveTangent=-0.6379, interpMode=RCIM_Cubic, leaveTangent=-0.6379, time=2.0828, value=0.6611}
           {arriveTangent=0.4353, interpMode=RCIM_Cubic, leaveTangent=0.4353, time=2.2158, value=0.9434}
           {arriveTangent=-0.3447, interpMode=RCIM_Cubic, leaveTangent=-0.3447, time=2.2673, value=0.7414}
           {arriveTangent=0.6603, interpMode=RCIM_Cubic, leaveTangent=0.6603, time=2.3258, value=0.9055}
           {arriveTangent=0.3801, interpMode=RCIM_Cubic, leaveTangent=0.3801, time=2.3971, value=0.8271}
           {arriveTangent=0.6687, interpMode=RCIM_Cubic, leaveTangent=0.6687, time=2.4654, value=0.9586}
           {arriveTangent=1.2527, interpMode=RCIM_Cubic, leaveTangent=1.2527, time=2.5143, value=0.9055}
           {arriveTangent=-0.515, interpMode=RCIM_Cubic, leaveTangent=-0.515, time=2.5566, value=1.0728}
           {arriveTangent=-1.6201, interpMode=RCIM_Cubic, leaveTangent=-1.6201, time=2.597, value=0.8629}
           {arriveTangent=-0.9669, interpMode=RCIM_Cubic, leaveTangent=-0.9669, time=2.6456, value=0.9286}
           {arriveTangent=0.1009, interpMode=RCIM_Cubic, leaveTangent=0.1009, time=2.6921, value=0.771}
           {arriveTangent=0.0929, interpMode=RCIM_Cubic, leaveTangent=0.0929, time=2.7509, value=0.9392}
           {arriveTangent=0.2939, interpMode=RCIM_Cubic, leaveTangent=0.2939, time=2.8654, value=0.7871}
           {arriveTangent=1.7227, interpMode=RCIM_Cubic, leaveTangent=1.7227, time=2.9271, value=0.991}
           {arriveTangent=0.1236, interpMode=RCIM_Cubic, leaveTangent=0.1236, time=2.957, value=0.9449}
           <at 3 1 RCIM_Cubic>
          ]
         }
        }}
       CurveIndex=$Emitter.NormalizedLoopAge
       Scale Curve=0.8
      )
      Input Min=&Add_Float(A=$System.FlickerCurveRemapMin, B=0)
      Input Max=1.2
      Output Min=0
      Output Max=1
     )
    )
   )
   /SetVariables(Particles.ComponentsEnabled:NiagaraBool=&MaskBoolBySpawnGroup(Masked Bool:NiagaraBool=false, Passthrough Bool:NiagaraBool=true, Spawn Group Mask:NiagaraInt32=0)) {isSetParameters=true}
   /SetVariables(Particles.LightPos:NiagaraPosition=$Particles.FN_Sprite_Attributes.Position) {isSetParameters=true}
  ]
  rendererClasses [
   &NiagaraSpriteRendererProperties
   &NiagaraLightRendererProperties
   &NiagaraComponentRendererProperties
  ]
  renderers [
   {rendererClass=&NiagaraSpriteRendererProperties}
   {rendererClass=&NiagaraSpriteRendererProperties, rendererIndex=1}
   {rendererClass=&NiagaraLightRendererProperties, rendererIndex=2}
   {rendererClass=&NiagaraComponentRendererProperties, rendererIndex=3}
  ]
 }
}
summary {
 systemName=NS_Backpack_NutmegMayo_Idle
 emitters [
  {
   emitterName=Embers001
   rendererClasses [
    &NiagaraSpriteRendererProperties
    &NiagaraLightRendererProperties
    &NiagaraComponentRendererProperties
   ]
  }
  {
   emitterName=Embers002
   rendererClasses [
    &NiagaraSpriteRendererProperties
    &NiagaraLightRendererProperties
    &NiagaraComponentRendererProperties
   ]
  }
  {
   emitterName=Embers003
   rendererClasses [
    &NiagaraSpriteRendererProperties
    &NiagaraLightRendererProperties
    &NiagaraComponentRendererProperties
   ]
  }
 ]
 userVariables [
  User.Color {type=LinearColor, description="", defaultValue {struct=LinearColor, value=<rgba 0 0.7624 1 1>}}
  User.LoopDuration {type=NiagaraFloat, defaultValue=10, description=""}
 ]
}
