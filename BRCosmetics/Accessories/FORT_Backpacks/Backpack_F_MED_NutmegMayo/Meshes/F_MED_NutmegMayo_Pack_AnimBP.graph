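# Anim Blueprint for the NutmegMayo backpack that spins two drill bones and adds simulated sway to big and small wire chains.
# Drill Rotate accumulates a wrapped spin angle, easing to a faster rate during melee, tactical sprint or sliding, applied before caching the pose.
# That pose feeds two AnimDynamics setups, skydive and default, selected by skydive state.
# Paired opposing additive wire rotations with differing blend-in times follow the blend.
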AnimBlueprint /BRCosmetics/Accessories/FORT_Backpacks/Backpack_F_MED_NutmegMayo/Meshes/F_MED_NutmegMayo_Pack_AnimBP
target_skeleton=&/F_MED_NutmegMayo_Pack_Skeleton
graphs [
 AnimGraph {nodes [
   n1 = Anim Dynamics - Bone: dyn_big_wire_01!(ComponentPose, bAlphaBoolEnabled:bool=True, GravityOverride:Vector="0,0,-400", ExternalForce="0,0,0") -> (Pose) {
    alphaInputType=Bool
    angularBiasOverride=1
    angularDampingOverride=1
    angularSpringConstant=1000
    bAngularSpring=true
    bOverrideAngularDamping=true
    bUseGravityOverride=true
    bUsePlanarLimit=false
    simulationSpace=RootRelative
    alphaBoolBlend {blendInTime=0.1, blendOption=Linear, blendOutTime=0.1}
    boundBone {boneName=dyn_big_wire_01}
    componentAppliedLinearAccClamp=<200 200 200>
    componentLinearAccScale=<0.3 0.3 0.3>
    componentLinearVelScale=<0.4 0.4 0.4>
    physicsBodyDefinitions [{
      collisionType=CustomSphere
      sphereCollisionRadius=2.5
      boundBone {boneName=dyn_big_wire_01}
      boxExtents=<3 3 3>
      constraintSetup {angularLimitsMax=<10 5 0>, angularLimitsMin=<-10 -15 0>, angularTarget=<0 0 0>}
      localJointOffset=<3 0 0>
     }]
   }
   n2 = Anim Dynamics - Bone: dyn_big_wire_02!(ComponentPose=@n1, bAlphaBoolEnabled:bool=True, GravityOverride:Vector="0,0,-300", ExternalForce="0,0,0") -> (Pose) {
    alphaInputType=Bool
    angularBiasOverride=1
    angularDampingOverride=1
    angularSpringConstant=1000
    bAngularSpring=true
    bOverrideAngularDamping=true
    bUseGravityOverride=true
    bUsePlanarLimit=false
    simulationSpace=RootRelative
    alphaBoolBlend {blendInTime=0.1, blendOption=Linear, blendOutTime=0.1}
    boundBone {boneName=dyn_big_wire_02}
    componentAppliedLinearAccClamp=<200 200 200>
    componentLinearAccScale=<0.3 0.3 0.3>
    componentLinearVelScale=<0.4 0.4 0.4>
    physicsBodyDefinitions [{
      collisionType=CustomSphere
      sphereCollisionRadius=2.5
      boundBone {boneName=dyn_big_wire_02}
      boxExtents=<3 3 3>
      constraintSetup {angularLimitsMax=<10 30 5>, angularLimitsMin=<-10 -40 -25>, angularTarget=<0 0 0>}
      localJointOffset=<3 0 0>
     }]
   }
   n3 = Anim Dynamics - Bone: dyn_big_wire_03!(ComponentPose=@n2, bAlphaBoolEnabled:bool=True, GravityOverride:Vector="0,0,-300", ExternalForce="0,0,0") -> (Pose) {
    alphaInputType=Bool
    angularBiasOverride=1
    angularDampingOverride=1
    angularSpringConstant=1000
    bAngularSpring=true
    bOverrideAngularDamping=true
    bUseGravityOverride=true
    bUsePlanarLimit=false
    simulationSpace=RootRelative
    alphaBoolBlend {blendInTime=0.1, blendOption=Linear, blendOutTime=0.1}
    boundBone {boneName=dyn_big_wire_03}
    componentAppliedLinearAccClamp=<200 200 200>
    componentLinearAccScale=<0.3 0.3 0.3>
    componentLinearVelScale=<0.4 0.4 0.4>
    physicsBodyDefinitions [{
      collisionType=CustomSphere
      sphereCollisionRadius=2.5
      boundBone {boneName=dyn_big_wire_03}
      boxExtents=<3 3 3>
      constraintSetup {angularLimitsMax=<15 35 10>, angularLimitsMin=<-15 -35 -25>, angularTarget=<0 0 0>}
      localJointOffset=<3 0 0>
     }]
   }
   n4 = Anim Dynamics - Bone: dyn_big_wire_04!(ComponentPose=@n3, bAlphaBoolEnabled:bool=True, GravityOverride:Vector="0,0,-300", ExternalForce="0,0,0") -> (Pose) {
    alphaInputType=Bool
    angularBiasOverride=1
    angularDampingOverride=1
    angularSpringConstant=1000
    bAngularSpring=true
    bOverrideAngularDamping=true
    bUseGravityOverride=true
    bUsePlanarLimit=false
    simulationSpace=RootRelative
    alphaBoolBlend {blendInTime=0.1, blendOption=Linear, blendOutTime=0.1}
    boundBone {boneName=dyn_big_wire_04}
    componentAppliedLinearAccClamp=<200 200 200>
    componentLinearAccScale=<0.3 0.3 0.3>
    componentLinearVelScale=<0.4 0.4 0.4>
    physicsBodyDefinitions [{
      collisionType=CustomSphere
      sphereCollisionRadius=2.5
      boundBone {boneName=dyn_big_wire_04}
      boxExtents=<3 3 3>
      constraintSetup {angularLimitsMax=<15 35 10>, angularLimitsMin=<-15 -35 -25>, angularTarget=<0 0 0>}
      localJointOffset=<3 0 0>
     }]
   }
   n5 = Anim Dynamics - Bone: dyn_small_wire_01!(ComponentPose=@n4, bAlphaBoolEnabled:bool=True, GravityOverride:Vector="0,0,-400", ExternalForce="0,0,0") -> (Pose) {
    alphaInputType=Bool
    angularBiasOverride=1
    angularDampingOverride=1
    angularSpringConstant=1000
    bAngularSpring=true
    bOverrideAngularDamping=true
    bUseGravityOverride=true
    bUsePlanarLimit=false
    simulationSpace=RootRelative
    alphaBoolBlend {blendInTime=0.1, blendOption=Linear, blendOutTime=0.1}
    boundBone {boneName=dyn_small_wire_01}
    componentAppliedLinearAccClamp=<200 200 200>
    componentLinearAccScale=<0.3 0.3 0.3>
    componentLinearVelScale=<0.4 0.4 0.4>
    physicsBodyDefinitions [{
      collisionType=CustomSphere
      sphereCollisionRadius=2.5
      boundBone {boneName=dyn_small_wire_01}
      boxExtents=<3 3 3>
      constraintSetup {angularLimitsMax=<5 0 0>, angularLimitsMin=<-5 0 -15>, angularTarget=<0 0 0>}
      localJointOffset=<3 0 0>
     }]
   }
   n6 = Anim Dynamics - Bone: dyn_small_wire_02!(ComponentPose=@n5, bAlphaBoolEnabled:bool=True, GravityOverride:Vector="0,0,-300", ExternalForce="0,0,0") -> (Pose) {
    alphaInputType=Bool
    angularBiasOverride=1
    angularDampingOverride=1
    angularSpringConstant=1000
    bAngularSpring=true
    bOverrideAngularDamping=true
    bUseGravityOverride=true
    bUsePlanarLimit=false
    simulationSpace=RootRelative
    alphaBoolBlend {blendInTime=0.1, blendOption=Linear, blendOutTime=0.1}
    boundBone {boneName=dyn_small_wire_02}
    componentAppliedLinearAccClamp=<200 200 200>
    componentLinearAccScale=<0.3 0.3 0.3>
    componentLinearVelScale=<0.4 0.4 0.4>
    physicsBodyDefinitions [{
      collisionType=CustomSphere
      sphereCollisionRadius=2.5
      boundBone {boneName=dyn_small_wire_02}
      boxExtents=<3 3 3>
      constraintSetup {angularLimitsMax=<5 0 8>, angularLimitsMin=<-5 0 -15>, angularTarget=<0 0 0>}
      localJointOffset=<3 0 0>
     }]
   }
   Anim Dynamics - Bone: dyn_small_wire_03!(ComponentPose=@n6, bAlphaBoolEnabled:bool=True, GravityOverride:Vector="0,0,-300", ExternalForce="0,0,0") -> (Pose) {
    alphaInputType=Bool
    angularBiasOverride=1
    angularDampingOverride=1
    angularSpringConstant=1000
    bAngularSpring=true
    bOverrideAngularDamping=true
    bUseGravityOverride=true
    bUsePlanarLimit=false
    simulationSpace=RootRelative
    alphaBoolBlend {blendInTime=0.1, blendOption=Linear, blendOutTime=0.1}
    boundBone {boneName=dyn_small_wire_03}
    componentAppliedLinearAccClamp=<200 200 200>
    componentLinearAccScale=<0.3 0.3 0.3>
    componentLinearVelScale=<0.4 0.4 0.4>
    physicsBodyDefinitions [{
      collisionType=CustomSphere
      sphereCollisionRadius=2.5
      boundBone {boneName=dyn_small_wire_03}
      boxExtents=<3 3 3>
      constraintSetup {angularLimitsMax=<5 0 8>, angularLimitsMin=<-5 0 -15>, angularTarget=<0 0 0>}
      localJointOffset=<3 0 0>
     }]
   }
   n7 = "Transform (Modify) Bone - Bone: dyn_big_wire_01"!(ComponentPose, bAlphaBoolEnabled:bool=True, Translation="0,0,0", Rotation="56.324,-175.252,161.253", Scale="1,1,1") -> (Pose) {
    alphaInputType=Bool
    rotationMode=BMM_Additive
    rotationSpace=BCS_BoneSpace
    translationMode=BMM_Additive
    translationSpace=BCS_BoneSpace
    alphaBoolBlend {blendInTime=3, blendOption=Linear, blendOutTime=0.5}
    boneToModify {boneName=dyn_big_wire_01}
    rotation=<rot 56.324 -175.252 161.253>
   }
   "Transform (Modify) Bone - Bone: dyn_big_wire_02"!(ComponentPose=@n7, bAlphaBoolEnabled:bool=True, Translation="0.12,1.916,0.982", Rotation="44.996,-36.675,-40.923", Scale="1,1,1") -> (Pose) {
    alphaInputType=Bool
    rotationMode=BMM_Additive
    rotationSpace=BCS_BoneSpace
    translationMode=BMM_Additive
    translationSpace=BCS_BoneSpace
    alphaBoolBlend {blendInTime=3, blendOption=Linear, blendOutTime=0.5}
    boneToModify {boneName=dyn_big_wire_02}
    rotation=<rot 44.996 -36.675 -40.923>
    translation=<0.12 1.916 0.982>
   }
   n8 = "Transform (Modify) Bone - Bone: dyn_big_wire_01"(bAlphaBoolEnabled=True, Translation="0,0,0", Rotation="-56.324,175.252,-161.253", Scale="1,1,1") {
    alphaInputType=Bool
    rotationMode=BMM_Additive
    rotationSpace=BCS_BoneSpace
    translationMode=BMM_Additive
    translationSpace=BCS_BoneSpace
    alphaBoolBlend {blendInTime=0.3, blendOption=Linear, blendOutTime=1}
    boneToModify {boneName=dyn_big_wire_01}
    rotation=<rot -56.324 175.252 -161.253>
   }
   "Transform (Modify) Bone - Bone: dyn_big_wire_02"(ComponentPose=@n8, bAlphaBoolEnabled=True, Translation="-0.120,1.916,-0.982", Rotation="-44.996,36.675,40.923", Scale="1,1,1") {
    alphaInputType=Bool
    rotationMode=BMM_Additive
    rotationSpace=BCS_BoneSpace
    translationMode=BMM_Additive
    translationSpace=BCS_BoneSpace
    alphaBoolBlend {blendInTime=0.3, blendOption=Linear, blendOutTime=1}
    boneToModify {boneName=dyn_big_wire_02}
    rotation=<rot -44.996 36.6749 40.923>
    translation=<-0.1198 1.9161 -0.9819>
   }
   n9 = Get bIsSkydiveDiveMode()
   n10 = Get bIsSkydiveFloating()
   n11 = Anim Dynamics - Bone: dyn_big_wire_01(
    ComponentPose=Local To Component(LocalPose=Use cached pose 'Presim'())
    bAlphaBoolEnabled=@n10
    GravityOverride="0,0,-400"
    ExternalForce="0,0,0"
   ) {
    alphaInputType=Bool
    angularBiasOverride=1
    angularDampingOverride=1
    angularSpringConstant=1000
    bAngularSpring=true
    bOverrideAngularDamping=true
    bUseGravityOverride=true
    bUsePlanarLimit=false
    simulationSpace=RootRelative
    alphaBoolBlend {blendInTime=0.1, blendOption=Linear, blendOutTime=0.1}
    boundBone {boneName=dyn_big_wire_01}
    componentAppliedLinearAccClamp=<200 200 200>
    componentLinearAccScale=<0.3 0.3 0.3>
    componentLinearVelScale=<0.4 0.4 0.4>
    physicsBodyDefinitions [{
      collisionType=CustomSphere
      sphereCollisionRadius=2.5
      boundBone {boneName=dyn_big_wire_01}
      boxExtents=<3 3 3>
      constraintSetup {angularLimitsMax=<10 80 10>, angularLimitsMin=<-10 25 -3>, angularTarget=<0 0 0>}
      localJointOffset=<3 0 0>
     }]
   }
   n12 = Anim Dynamics - Bone: dyn_big_wire_02(ComponentPose=@n11, bAlphaBoolEnabled=@n10, GravityOverride="0,0,-300", ExternalForce="0,0,0") {
    alphaInputType=Bool
    angularBiasOverride=1
    angularDampingOverride=1
    angularSpringConstant=1000
    bAngularSpring=true
    bOverrideAngularDamping=true
    bUseGravityOverride=true
    bUsePlanarLimit=false
    simulationSpace=RootRelative
    alphaBoolBlend {blendInTime=0.1, blendOption=Linear, blendOutTime=0.1}
    boundBone {boneName=dyn_big_wire_02}
    componentAppliedLinearAccClamp=<200 200 200>
    componentLinearAccScale=<0.3 0.3 0.3>
    componentLinearVelScale=<0.4 0.4 0.4>
    physicsBodyDefinitions [{
      collisionType=CustomSphere
      sphereCollisionRadius=2.5
      boundBone {boneName=dyn_big_wire_02}
      boxExtents=<3 3 3>
      constraintSetup {angularLimitsMax=<20 50 30>, angularLimitsMin=<-20 35 -30>, angularTarget=<0 0 0>}
      localJointOffset=<3 0 0>
     }]
   }
   n13 = Anim Dynamics - Bone: dyn_big_wire_03(ComponentPose=@n12, bAlphaBoolEnabled=@n10, GravityOverride="0,0,-200", ExternalForce="0,0,0") {
    alphaInputType=Bool
    angularBiasOverride=1
    angularDampingOverride=1
    angularSpringConstant=1000
    bAngularSpring=true
    bOverrideAngularDamping=true
    bUseGravityOverride=true
    bUsePlanarLimit=false
    simulationSpace=RootRelative
    alphaBoolBlend {blendInTime=0.1, blendOption=Linear, blendOutTime=0.1}
    boundBone {boneName=dyn_big_wire_03}
    componentAppliedLinearAccClamp=<200 200 200>
    componentLinearAccScale=<0.3 0.3 0.3>
    componentLinearVelScale=<0.4 0.4 0.4>
    physicsBodyDefinitions [{
      collisionType=CustomSphere
      sphereCollisionRadius=2.5
      boundBone {boneName=dyn_big_wire_03}
      boxExtents=<3 3 3>
      constraintSetup {angularLimitsMax=<30 50 30>, angularLimitsMin=<0 35 -30>, angularTarget=<0 0 0>}
      localJointOffset=<3 0 0>
     }]
   }
   n14 = Anim Dynamics - Bone: dyn_big_wire_04(ComponentPose=@n13, bAlphaBoolEnabled=@n10, GravityOverride="0,0,-100", ExternalForce="0,0,0") {
    alphaInputType=Bool
    angularBiasOverride=1
    angularDampingOverride=1
    angularSpringConstant=1000
    bAngularSpring=true
    bOverrideAngularDamping=true
    bUseGravityOverride=true
    bUsePlanarLimit=false
    simulationSpace=RootRelative
    alphaBoolBlend {blendInTime=0.1, blendOption=Linear, blendOutTime=0.1}
    boundBone {boneName=dyn_big_wire_04}
    componentAppliedLinearAccClamp=<200 200 200>
    componentLinearAccScale=<0.3 0.3 0.3>
    componentLinearVelScale=<0.4 0.4 0.4>
    physicsBodyDefinitions [{
      collisionType=CustomSphere
      sphereCollisionRadius=2.5
      boundBone {boneName=dyn_big_wire_04}
      boxExtents=<3 3 3>
      constraintSetup {angularLimitsMax=<30 50 30>, angularLimitsMin=<-30 30 -25>, angularTarget=<0 0 0>}
      localJointOffset=<3 0 0>
     }]
   }
   n15 = Anim Dynamics - Bone: dyn_small_wire_01(ComponentPose=@n14, bAlphaBoolEnabled=@n10, GravityOverride="0,0,-400", ExternalForce="0,0,0") {
    alphaInputType=Bool
    angularBiasOverride=1
    angularDampingOverride=1
    angularSpringConstant=1000
    bAngularSpring=true
    bOverrideAngularDamping=true
    bUseGravityOverride=true
    bUsePlanarLimit=false
    simulationSpace=RootRelative
    alphaBoolBlend {blendInTime=0.1, blendOption=Linear, blendOutTime=0.1}
    boundBone {boneName=dyn_small_wire_01}
    componentAppliedLinearAccClamp=<200 200 200>
    componentLinearAccScale=<0.3 0.3 0.3>
    componentLinearVelScale=<0.4 0.4 0.4>
    physicsBodyDefinitions [{
      collisionType=CustomSphere
      sphereCollisionRadius=2.5
      boundBone {boneName=dyn_small_wire_01}
      boxExtents=<3 3 3>
      constraintSetup {angularLimitsMax=<5 50 0>, angularLimitsMin=<-5 10 -15>, angularTarget=<0 0 0>}
      localJointOffset=<3 0 0>
     }]
   }
   n16 = Anim Dynamics - Bone: dyn_small_wire_02(ComponentPose=@n15, bAlphaBoolEnabled=@n10, GravityOverride="0,0,-300", ExternalForce="0,0,0") {
    alphaInputType=Bool
    angularBiasOverride=1
    angularDampingOverride=1
    angularSpringConstant=1000
    bAngularSpring=true
    bOverrideAngularDamping=true
    bUseGravityOverride=true
    bUsePlanarLimit=false
    simulationSpace=RootRelative
    alphaBoolBlend {blendInTime=0.1, blendOption=Linear, blendOutTime=0.1}
    boundBone {boneName=dyn_small_wire_02}
    componentAppliedLinearAccClamp=<200 200 200>
    componentLinearAccScale=<0.3 0.3 0.3>
    componentLinearVelScale=<0.4 0.4 0.4>
    physicsBodyDefinitions [{
      collisionType=CustomSphere
      sphereCollisionRadius=2.5
      boundBone {boneName=dyn_small_wire_02}
      boxExtents=<3 3 3>
      constraintSetup {angularLimitsMax=<20 25 20>, angularLimitsMin=<-20 10 -15>, angularTarget=<0 0 0>}
      localJointOffset=<3 0 0>
     }]
   }
   n17 = Anim Dynamics - Bone: dyn_small_wire_03(ComponentPose=@n16, bAlphaBoolEnabled=@n10, GravityOverride="0,0,-300", ExternalForce="0,0,0") {
    alphaInputType=Bool
    angularBiasOverride=1
    angularDampingOverride=1
    angularSpringConstant=1000
    bAngularSpring=true
    bOverrideAngularDamping=true
    bUseGravityOverride=true
    bUsePlanarLimit=false
    simulationSpace=RootRelative
    alphaBoolBlend {blendInTime=0.1, blendOption=Linear, blendOutTime=0.1}
    boundBone {boneName=dyn_small_wire_03}
    componentAppliedLinearAccClamp=<200 200 200>
    componentLinearAccScale=<0.3 0.3 0.3>
    componentLinearVelScale=<0.4 0.4 0.4>
    physicsBodyDefinitions [{
      collisionType=CustomSphere
      sphereCollisionRadius=2.5
      boundBone {boneName=dyn_small_wire_03}
      boxExtents=<3 3 3>
      constraintSetup {angularLimitsMax=<20 30 20>, angularLimitsMin=<-20 10 -15>, angularTarget=<0 0 0>}
      localJointOffset=<3 0 0>
     }]
   }
   n18 = Anim Dynamics - Bone: dyn_big_wire_01(ComponentPose=@n17, bAlphaBoolEnabled=@n9, GravityOverride="0,0,-400", ExternalForce="0,0,0") {
    alphaInputType=Bool
    angularBiasOverride=1
    angularDampingOverride=1
    angularSpringConstant=1000
    bAngularSpring=true
    bOverrideAngularDamping=true
    bUseGravityOverride=true
    bUsePlanarLimit=false
    simulationSpace=RootRelative
    alphaBoolBlend {blendInTime=0.1, blendOption=Linear, blendOutTime=0.1}
    boundBone {boneName=dyn_big_wire_01}
    componentAppliedLinearAccClamp=<200 200 200>
    componentLinearAccScale=<0.3 0.3 0.3>
    componentLinearVelScale=<0.4 0.4 0.4>
    physicsBodyDefinitions [{
      collisionType=CustomSphere
      sphereCollisionRadius=2.5
      boundBone {boneName=dyn_big_wire_01}
      boxExtents=<3 3 3>
      constraintSetup {angularLimitsMax=<5 5 5>, angularLimitsMin=<-5 -5 -3>, angularTarget=<0 0 0>}
      localJointOffset=<3 0 0>
     }]
   }
   n19 = Anim Dynamics - Bone: dyn_big_wire_02(ComponentPose=@n18, bAlphaBoolEnabled=@n9, GravityOverride="0,0,-300", ExternalForce="0,0,0") {
    alphaInputType=Bool
    angularBiasOverride=1
    angularDampingOverride=1
    angularSpringConstant=1000
    bAngularSpring=true
    bOverrideAngularDamping=true
    bUseGravityOverride=true
    bUsePlanarLimit=false
    simulationSpace=RootRelative
    alphaBoolBlend {blendInTime=0.1, blendOption=Linear, blendOutTime=0.1}
    boundBone {boneName=dyn_big_wire_02}
    componentAppliedLinearAccClamp=<200 200 200>
    componentLinearAccScale=<0.3 0.3 0.3>
    componentLinearVelScale=<0.4 0.4 0.4>
    physicsBodyDefinitions [{
      collisionType=CustomSphere
      sphereCollisionRadius=2.5
      boundBone {boneName=dyn_big_wire_02}
      boxExtents=<3 3 3>
      constraintSetup {angularLimitsMax=<5 5 5>, angularLimitsMin=<-5 -5 -3>, angularTarget=<0 0 0>}
      localJointOffset=<3 0 0>
     }]
   }
   n20 = Anim Dynamics - Bone: dyn_big_wire_03(ComponentPose=@n19, bAlphaBoolEnabled=@n9, GravityOverride="0,0,-200", ExternalForce="0,0,0") {
    alphaInputType=Bool
    angularBiasOverride=1
    angularDampingOverride=1
    angularSpringConstant=1000
    bAngularSpring=true
    bOverrideAngularDamping=true
    bUseGravityOverride=true
    bUsePlanarLimit=false
    simulationSpace=RootRelative
    alphaBoolBlend {blendInTime=0.1, blendOption=Linear, blendOutTime=0.1}
    boundBone {boneName=dyn_big_wire_03}
    componentAppliedLinearAccClamp=<200 200 200>
    componentLinearAccScale=<0.3 0.3 0.3>
    componentLinearVelScale=<0.4 0.4 0.4>
    physicsBodyDefinitions [{
      collisionType=CustomSphere
      sphereCollisionRadius=2.5
      boundBone {boneName=dyn_big_wire_03}
      boxExtents=<3 3 3>
      constraintSetup {angularLimitsMax=<5 5 5>, angularLimitsMin=<-5 -5 -3>, angularTarget=<0 0 0>}
      localJointOffset=<3 0 0>
     }]
   }
   n21 = Anim Dynamics - Bone: dyn_big_wire_04(ComponentPose=@n20, bAlphaBoolEnabled=@n9, GravityOverride="0,0,-100", ExternalForce="0,0,0") {
    alphaInputType=Bool
    angularBiasOverride=1
    angularDampingOverride=1
    angularSpringConstant=1000
    bAngularSpring=true
    bOverrideAngularDamping=true
    bUseGravityOverride=true
    bUsePlanarLimit=false
    simulationSpace=RootRelative
    alphaBoolBlend {blendInTime=0.1, blendOption=Linear, blendOutTime=0.1}
    boundBone {boneName=dyn_big_wire_04}
    componentAppliedLinearAccClamp=<200 200 200>
    componentLinearAccScale=<0.3 0.3 0.3>
    componentLinearVelScale=<0.4 0.4 0.4>
    physicsBodyDefinitions [{
      collisionType=CustomSphere
      sphereCollisionRadius=2.5
      boundBone {boneName=dyn_big_wire_04}
      boxExtents=<3 3 3>
      constraintSetup {angularLimitsMax=<5 5 5>, angularLimitsMin=<-5 -5 -3>, angularTarget=<0 0 0>}
      localJointOffset=<3 0 0>
     }]
   }
   n22 = Anim Dynamics - Bone: dyn_small_wire_01(ComponentPose=@n21, bAlphaBoolEnabled=@n9, GravityOverride="0,0,-400", ExternalForce="0,0,0") {
    alphaInputType=Bool
    angularBiasOverride=1
    angularDampingOverride=1
    angularSpringConstant=1000
    bAngularSpring=true
    bOverrideAngularDamping=true
    bUseGravityOverride=true
    bUsePlanarLimit=false
    simulationSpace=RootRelative
    alphaBoolBlend {blendInTime=0.1, blendOption=Linear, blendOutTime=0.1}
    boundBone {boneName=dyn_small_wire_01}
    componentAppliedLinearAccClamp=<200 200 200>
    componentLinearAccScale=<0.3 0.3 0.3>
    componentLinearVelScale=<0.4 0.4 0.4>
    physicsBodyDefinitions [{
      collisionType=CustomSphere
      sphereCollisionRadius=2.5
      boundBone {boneName=dyn_small_wire_01}
      boxExtents=<3 3 3>
      constraintSetup {angularLimitsMax=<5 5 5>, angularLimitsMin=<-5 -5 -3>, angularTarget=<0 0 0>}
      localJointOffset=<3 0 0>
     }]
   }
   n23 = Anim Dynamics - Bone: dyn_small_wire_02(ComponentPose=@n22, bAlphaBoolEnabled=@n9, GravityOverride="0,0,-300", ExternalForce="0,0,0") {
    alphaInputType=Bool
    angularBiasOverride=1
    angularDampingOverride=1
    angularSpringConstant=1000
    bAngularSpring=true
    bOverrideAngularDamping=true
    bUseGravityOverride=true
    bUsePlanarLimit=false
    simulationSpace=RootRelative
    alphaBoolBlend {blendInTime=0.1, blendOption=Linear, blendOutTime=0.1}
    boundBone {boneName=dyn_small_wire_02}
    componentAppliedLinearAccClamp=<200 200 200>
    componentLinearAccScale=<0.3 0.3 0.3>
    componentLinearVelScale=<0.4 0.4 0.4>
    physicsBodyDefinitions [{
      collisionType=CustomSphere
      sphereCollisionRadius=2.5
      boundBone {boneName=dyn_small_wire_02}
      boxExtents=<3 3 3>
      constraintSetup {angularLimitsMax=<5 5 5>, angularLimitsMin=<-5 -5 -3>, angularTarget=<0 0 0>}
      localJointOffset=<3 0 0>
     }]
   }
   n24 = Anim Dynamics - Bone: dyn_small_wire_03(ComponentPose=@n23, bAlphaBoolEnabled=@n9, GravityOverride="0,0,-300", ExternalForce="0,0,0") {
    alphaInputType=Bool
    angularBiasOverride=1
    angularDampingOverride=1
    angularSpringConstant=1000
    bAngularSpring=true
    bOverrideAngularDamping=true
    bUseGravityOverride=true
    bUsePlanarLimit=false
    simulationSpace=RootRelative
    alphaBoolBlend {blendInTime=0.1, blendOption=Linear, blendOutTime=0.1}
    boundBone {boneName=dyn_small_wire_03}
    componentAppliedLinearAccClamp=<200 200 200>
    componentLinearAccScale=<0.3 0.3 0.3>
    componentLinearVelScale=<0.4 0.4 0.4>
    physicsBodyDefinitions [{
      collisionType=CustomSphere
      sphereCollisionRadius=2.5
      boundBone {boneName=dyn_small_wire_03}
      boxExtents=<3 3 3>
      constraintSetup {angularLimitsMax=<5 5 5>, angularLimitsMin=<-5 -5 -3>, angularTarget=<0 0 0>}
      localJointOffset=<3 0 0>
     }]
   }
   n25 = "Transform (Modify) Bone - Bone: pack"!(ComponentPose=@n24, bAlphaBoolEnabled:bool, Translation="0,-1.5,-2.1", Rotation="0,0,0", Scale) -> (Pose) {
    alphaInputType=Bool
    rotationMode=BMM_Replace
    rotationSpace=BCS_ParentBoneSpace
    translationMode=BMM_Additive
    translationSpace=BCS_BoneSpace
    alphaBoolBlend {blendInTime=0.1, blendOption=Linear, blendOutTime=0.1}
    boneToModify {boneName=pack}
    translation {y=-1.5, z=-2.1}
   }
   n26 = Component To Local(ComponentPose=@n25)
   n27 = Anim Dynamics - Bone: dyn_big_wire_01(
    ComponentPose=Local To Component(LocalPose=Use cached pose 'Presim'())
    bAlphaBoolEnabled=True
    GravityOverride="0,0,-900"
    ExternalForce="0,0,0"
   ) {
    alphaInputType=Bool
    angularBiasOverride=1
    angularDampingOverride=1
    angularSpringConstant=1000
    bAngularSpring=true
    bOverrideAngularDamping=true
    bUseGravityOverride=true
    bUsePlanarLimit=false
    simulationSpace=RootRelative
    alphaBoolBlend {blendInTime=0.1, blendOption=Linear, blendOutTime=0.1}
    boundBone {boneName=dyn_big_wire_01}
    componentAppliedLinearAccClamp=<200 200 200>
    componentLinearAccScale=<0.3 0.3 0.3>
    componentLinearVelScale=<0.4 0.4 0.4>
    physicsBodyDefinitions [{
      collisionType=CustomSphere
      sphereCollisionRadius=2.5
      boundBone {boneName=dyn_big_wire_01}
      boxExtents=<7 7 7>
      constraintSetup {angularLimitsMax=<10 15 10>, angularLimitsMin=<-10 -5 -3>, angularTarget=<0 0 0>}
      localJointOffset=<1 0 0>
     }]
   }
   n28 = Anim Dynamics - Bone: dyn_big_wire_02(ComponentPose=@n27, bAlphaBoolEnabled=True, GravityOverride="0,0,-500", ExternalForce="0,0,0") {
    alphaInputType=Bool
    angularBiasOverride=1
    angularDampingOverride=1
    angularSpringConstant=1000
    bAngularSpring=true
    bOverrideAngularDamping=true
    bUseGravityOverride=true
    bUsePlanarLimit=false
    simulationSpace=RootRelative
    alphaBoolBlend {blendInTime=0.1, blendOption=Linear, blendOutTime=0.1}
    boundBone {boneName=dyn_big_wire_02}
    componentAppliedLinearAccClamp=<200 200 200>
    componentLinearAccScale=<0.3 0.3 0.3>
    componentLinearVelScale=<0.4 0.4 0.4>
    physicsBodyDefinitions [{
      collisionType=CustomSphere
      sphereCollisionRadius=2.5
      boundBone {boneName=dyn_big_wire_02}
      boxExtents=<7 7 7>
      constraintSetup {angularLimitsMax=<10 30 5>, angularLimitsMin=<-10 -40 -25>, angularTarget=<0 0 0>}
      localJointOffset=<1 0 0>
     }]
   }
   n29 = Anim Dynamics - Bone: dyn_big_wire_03(ComponentPose=@n28, bAlphaBoolEnabled=True, GravityOverride="0,0,-500", ExternalForce="0,0,0") {
    alphaInputType=Bool
    angularBiasOverride=1
    angularDampingOverride=1
    angularSpringConstant=1000
    bAngularSpring=true
    bOverrideAngularDamping=true
    bUseGravityOverride=true
    bUsePlanarLimit=false
    simulationSpace=RootRelative
    alphaBoolBlend {blendInTime=0.1, blendOption=Linear, blendOutTime=0.1}
    boundBone {boneName=dyn_big_wire_03}
    componentAppliedLinearAccClamp=<200 200 200>
    componentLinearAccScale=<0.3 0.3 0.3>
    componentLinearVelScale=<0.4 0.4 0.4>
    physicsBodyDefinitions [{
      collisionType=CustomSphere
      sphereCollisionRadius=2.5
      boundBone {boneName=dyn_big_wire_03}
      boxExtents=<5.5 5.5 5.5>
      constraintSetup {angularLimitsMax=<15 35 10>, angularLimitsMin=<-15 -35 -25>, angularTarget=<0 0 0>}
      localJointOffset=<1 0 0>
     }]
   }
   n30 = Anim Dynamics - Bone: dyn_big_wire_04(ComponentPose=@n29, bAlphaBoolEnabled=True, GravityOverride="0,0,-500", ExternalForce="0,0,0") {
    alphaInputType=Bool
    angularBiasOverride=1
    angularDampingOverride=1
    angularSpringConstant=1000
    bAngularSpring=true
    bOverrideAngularDamping=true
    bUseGravityOverride=true
    bUsePlanarLimit=false
    simulationSpace=RootRelative
    alphaBoolBlend {blendInTime=0.1, blendOption=Linear, blendOutTime=0.1}
    boundBone {boneName=dyn_big_wire_04}
    componentAppliedLinearAccClamp=<200 200 200>
    componentLinearAccScale=<0.3 0.3 0.3>
    componentLinearVelScale=<0.4 0.4 0.4>
    physicsBodyDefinitions [{
      collisionType=CustomSphere
      sphereCollisionRadius=2.5
      boundBone {boneName=dyn_big_wire_04}
      boxExtents=<5.5 5.5 5.5>
      constraintSetup {angularLimitsMax=<15 35 10>, angularLimitsMin=<-15 -35 -25>, angularTarget=<0 0 0>}
      localJointOffset=<1 0 0>
     }]
   }
   n31 = Anim Dynamics - Bone: dyn_small_wire_01(ComponentPose=@n30, bAlphaBoolEnabled=True, GravityOverride="0,0,-400", ExternalForce="0,0,0") {
    alphaInputType=Bool
    angularBiasOverride=1
    angularDampingOverride=1
    angularSpringConstant=1000
    bAngularSpring=true
    bOverrideAngularDamping=true
    bUseGravityOverride=true
    bUsePlanarLimit=false
    simulationSpace=RootRelative
    alphaBoolBlend {blendInTime=0.1, blendOption=Linear, blendOutTime=0.1}
    boundBone {boneName=dyn_small_wire_01}
    componentAppliedLinearAccClamp=<200 200 200>
    componentLinearAccScale=<0.3 0.3 0.3>
    componentLinearVelScale=<0.4 0.4 0.4>
    physicsBodyDefinitions [{
      collisionType=CustomSphere
      sphereCollisionRadius=2.5
      boundBone {boneName=dyn_small_wire_01}
      boxExtents=<4.5 4.5 4.5>
      constraintSetup {angularLimitsMax=<5 10 0>, angularLimitsMin=<-5 0 -15>, angularTarget=<0 0 0>}
      localJointOffset=<1 0 0>
     }]
   }
   n32 = Anim Dynamics - Bone: dyn_small_wire_02(ComponentPose=@n31, bAlphaBoolEnabled=True, GravityOverride="0,0,-300", ExternalForce="0,0,0") {
    alphaInputType=Bool
    angularBiasOverride=1
    angularDampingOverride=1
    angularSpringConstant=1000
    bAngularSpring=true
    bOverrideAngularDamping=true
    bUseGravityOverride=true
    bUsePlanarLimit=false
    simulationSpace=RootRelative
    alphaBoolBlend {blendInTime=0.1, blendOption=Linear, blendOutTime=0.1}
    boundBone {boneName=dyn_small_wire_02}
    componentAppliedLinearAccClamp=<200 200 200>
    componentLinearAccScale=<0.3 0.3 0.3>
    componentLinearVelScale=<0.4 0.4 0.4>
    physicsBodyDefinitions [{
      collisionType=CustomSphere
      sphereCollisionRadius=2.5
      boundBone {boneName=dyn_small_wire_02}
      boxExtents=<4.5 4.5 4.5>
      constraintSetup {angularLimitsMax=<5 10 8>, angularLimitsMin=<-5 -10 -15>, angularTarget=<0 0 0>}
      localJointOffset=<1 0 0>
     }]
   }
   n33 = Anim Dynamics - Bone: dyn_small_wire_03(ComponentPose=@n32, bAlphaBoolEnabled=True, GravityOverride="0,0,-300", ExternalForce="0,0,0") {
    alphaInputType=Bool
    angularBiasOverride=1
    angularDampingOverride=1
    angularSpringConstant=1000
    bAngularSpring=true
    bOverrideAngularDamping=true
    bUseGravityOverride=true
    bUsePlanarLimit=false
    simulationSpace=RootRelative
    alphaBoolBlend {blendInTime=0.1, blendOption=Linear, blendOutTime=0.1}
    boundBone {boneName=dyn_small_wire_03}
    componentAppliedLinearAccClamp=<200 200 200>
    componentLinearAccScale=<0.3 0.3 0.3>
    componentLinearVelScale=<0.4 0.4 0.4>
    physicsBodyDefinitions [{
      collisionType=CustomSphere
      sphereCollisionRadius=2.5
      boundBone {boneName=dyn_small_wire_03}
      boxExtents=<4.5 4.5 4.5>
      constraintSetup {angularLimitsMax=<5 15 8>, angularLimitsMin=<-5 -15 -15>, angularTarget=<0 0 0>}
      localJointOffset=<1 0 0>
     }]
   }
   n34 = Component To Local(ComponentPose=@n33)
   n35 = Get Backpack offset() -> (Backpack offset:bool)
   n36 = "Transform (Modify) Bone - Bone: dyn_big_wire_01"(
    ComponentPose=Local To Component(LocalPose=Blend Poses by bool(BlendPose_0=@n26, BlendPose_1=@n34, bActiveValue=@n35) {blendType=Linear})
    bAlphaBoolEnabled=True
    Translation="0,0,0"
    Rotation="0,80,0"
    Scale="1,1,1"
   ) {
    alphaInputType=Bool
    rotationMode=BMM_Additive
    rotationSpace=BCS_BoneSpace
    translationMode=BMM_Additive
    translationSpace=BCS_BoneSpace
    alphaBoolBlend {blendInTime=0.5, blendOption=Linear, blendOutTime=1}
    boneToModify {boneName=dyn_big_wire_01}
    rotation {yaw=80}
   }
   n37 = "Transform (Modify) Bone - Bone: dyn_big_wire_02"(ComponentPose=@n36, bAlphaBoolEnabled=True, Translation="0,0,0", Rotation="0,80,0", Scale="1,1,1") {
    alphaInputType=Bool
    rotationMode=BMM_Additive
    rotationSpace=BCS_BoneSpace
    translationMode=BMM_Additive
    translationSpace=BCS_BoneSpace
    alphaBoolBlend {blendInTime=0.5, blendOption=Linear, blendOutTime=1}
    boneToModify {boneName=dyn_big_wire_02}
    rotation {yaw=80}
   }
   n38 = "Transform (Modify) Bone - Bone: dyn_small_wire_01"!(ComponentPose=@n37, bAlphaBoolEnabled:bool=True, Translation="0,0,0", Rotation="-70.000,0.000,-0.000", Scale="1,1,1") -> (Pose) {
    alphaInputType=Bool
    rotationMode=BMM_Additive
    translationMode=BMM_Additive
    translationSpace=BCS_BoneSpace
    alphaBoolBlend {blendInTime=0.5, blendOption=Linear, blendOutTime=1}
    boneToModify {boneName=dyn_small_wire_01}
    rotation {pitch=-70}
   }
   n39 = "Transform (Modify) Bone - Bone: dyn_small_wire_02"!(ComponentPose=@n38, bAlphaBoolEnabled:bool=True, Translation="0,0,0", Rotation="-50.000,0.000,-0.000", Scale="1,1,1") -> (Pose) {
    alphaInputType=Bool
    rotationMode=BMM_Additive
    translationMode=BMM_Additive
    translationSpace=BCS_BoneSpace
    alphaBoolBlend {blendInTime=0.5, blendOption=Linear, blendOutTime=1}
    boneToModify {boneName=dyn_small_wire_02}
    rotation {pitch=-50}
   }
   n40 = "Transform (Modify) Bone - Bone: dyn_big_wire_01"(ComponentPose=@n39, bAlphaBoolEnabled=True, Translation="0,0,0", Rotation="0,-80,0", Scale="1,1,1") {
    alphaInputType=Bool
    rotationMode=BMM_Additive
    rotationSpace=BCS_BoneSpace
    translationMode=BMM_Additive
    translationSpace=BCS_BoneSpace
    alphaBoolBlend {blendInTime=3, blendOption=Linear, blendOutTime=0.5}
    boneToModify {boneName=dyn_big_wire_01}
    rotation {yaw=-80}
   }
   n41 = "Transform (Modify) Bone - Bone: dyn_big_wire_02"(ComponentPose=@n40, bAlphaBoolEnabled=True, Translation="0,0,0", Rotation="0,-80,0", Scale="1,1,1") {
    alphaInputType=Bool
    rotationMode=BMM_Additive
    rotationSpace=BCS_BoneSpace
    translationMode=BMM_Additive
    translationSpace=BCS_BoneSpace
    alphaBoolBlend {blendInTime=3, blendOption=Linear, blendOutTime=0.5}
    boneToModify {boneName=dyn_big_wire_02}
    rotation {yaw=-80}
   }
   n42 = "Transform (Modify) Bone - Bone: dyn_small_wire_01"(ComponentPose=@n41, bAlphaBoolEnabled=True, Translation="0,0,0", Rotation="70,0,0", Scale="1,1,1") {
    alphaInputType=Bool
    rotationMode=BMM_Additive
    translationMode=BMM_Additive
    translationSpace=BCS_BoneSpace
    alphaBoolBlend {blendInTime=3, blendOption=Linear, blendOutTime=0.5}
    boneToModify {boneName=dyn_small_wire_01}
    rotation {pitch=70}
   }
   n43 = "Transform (Modify) Bone - Bone: dyn_small_wire_02"(ComponentPose=@n42, bAlphaBoolEnabled=True, Translation="0,0,0", Rotation="50,0,0", Scale="1,1,1") {
    alphaInputType=Bool
    rotationMode=BMM_Additive
    translationMode=BMM_Additive
    translationSpace=BCS_BoneSpace
    alphaBoolBlend {blendInTime=3, blendOption=Linear, blendOutTime=0.5}
    boneToModify {boneName=dyn_small_wire_02}
    rotation {pitch=50}
   }
   Output Pose(Result=Component To Local(ComponentPose=@n43)) {layerGroup=None}
   n44 = "Transform (Modify) Bone - Bone: drill_top_01"(
    ComponentPose=Local To Component(LocalPose=Input Pose())
    Alpha=1
    Translation="0,0,0"
    Rotation=Get Spin result() -> (Spin result:Rotator)
    Scale="1,1,1"
   ) {rotationMode=BMM_Additive, rotationSpace=BCS_BoneSpace, alphaBoolBlend {blendOption=Linear}, boneToModify {boneName=drill_top_01}}
   n45 = "Transform (Modify) Bone - Bone: drill_bot_01"(ComponentPose=@n44, Alpha=1, Translation="0,0,0", Rotation=Get Spin result(), Scale="1,1,1") {rotationMode=BMM_Additive, rotationSpace=BCS_BoneSpace, alphaBoolBlend {blendOption=Linear}, boneToModify {boneName=drill_bot_01}}
   Save cached pose 'Presim'(Pose=Component To Local(ComponentPose=@n45)) {cacheName=Presim}
  ]}
 Drill Rotate {nodes [
   n1 = Drill Rotate!() -> (then, Spin:double=~, Speed:double=~) {is=title K2Node_FunctionEntry}
   > n2 = Set Spin var(Spin var:double=@n1) -> (Output_Get:double)
   n3 = OR Boolean!(A=Get bIsPlayingMeleeAnim(), B=Get bIsTacticalSprinting(), C:bool=Get bIsSliding()) -> (ReturnValue)
   > Set Acceleration(Acceleration:double=FInterpTo(
     Current=Get Acceleration() -> (Acceleration:double)
     Target=SelectFloat(A=1.2, B=0.12, bPickA=@n3)
     DeltaTime=Get Delta Time X()
     InterpSpeed=3
    )) -> (Output_Get:double)
   > n4 = Branch(Condition="float >= float"(A=@n2, B=720))
   > n5 = Set Spin var(Spin var=float - float(A=Get Spin var() -> (Spin var:double), B=720))
   n6 = Get Velocity mult() -> (Velocity mult:double)
   n7 = Get Acceleration()
   n8 = Get Delta time remap() -> (Delta time remap:double)
   n9 = Set Spin var(
    execute=@n4.else
    Spin var=float + float!(A=Get Spin var(), B=float * float(A=@n1.Speed, B=@n8) {is=title K2Node_PromotableOperator}, C:double=float * float(A=@n7, B=@n6)) -> (ReturnValue)
   )
   Return Node!(execute=[@n9.then, @n5.then], Output:double=@n9)
   OR Boolean!(A=Get bIsPlayingMeleeAnim(), B=Get bIsSprinting(), C=Get bIsSliding(), D:bool=Get bIsJumping()) -> (ReturnValue)
  ]}
 EventGraph {nodes [
   n1 = Event BlueprintUpdateAnimation()
   > Set Delta Time X(Delta Time X=@n1.DeltaTimeX)
   > n2 = Sequence()
   > Set Delta time remap(Delta time remap:double=FInterpTo(Current=Get Delta time remap(), Target=1, DeltaTime=0, InterpSpeed=1)) -> (Output_Get:double)
   > Set Backpack offset(Backpack offset:bool=OR Boolean!(A=Get bIsSkydiveDiveMode(), B=Get bIsSkydiveFloating()) -> (ReturnValue)) -> (Output_Get:bool)
   n3 = Drill Rotate!(execute=@n2.then_1, self, Spin:double=Get Spin() -> (Spin:double), Speed:double=Get Speed() -> (Speed:double)) -> (then, Output:double) {is=title K2Node_CallFunction}
   > n4 = Set Spin(Spin:double=@n3) -> (Output_Get:double)
   > Set Spin result(Spin result:Rotator=MakeRotator(Roll=@n4)) -> (Output_Get:Rotator)
   TryGetPawnOwner()
  ]}
]
variables [
 Acceleration {type=number}
 Backpack offset {type=boolean}
 Delta time remap {type=number}
 Delta Time X {type=number}
 Speed {type=number}
 Spin {type=number}
 Spin result {type=Rotator}
 Spin var {type=number}
 Velocity mult {type=number, default=12}
]
